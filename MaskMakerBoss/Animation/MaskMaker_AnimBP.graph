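# Animation Blueprint for the MaskMaker boss on the Fortnite average male player skeleton, switching between rest and combat idle poses.
# Base SM starts in Rest, enters Combat Idle on bEnterCombatIdle and returns on bIsUnaware.
# Its output chains through DefaultSlot, FullBody and AdditiveFullBody montage slots to the final pose.
# The event graph is empty and neither transition flag is declared or set here.
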AnimBlueprint /MaskMakerBoss/Animation/MaskMaker_AnimBP
target_skeleton=&Fortnite_M_Avg_Player_Skeleton
graphs [
 AnimGraph {nodes [
   AnimGraphNode_StateMachine_0 = Base SM() {is=title AnimGraphNode_StateMachine}
   n1 = Slot 'DefaultSlot'(Source=@AnimGraphNode_StateMachine_0) {bAlwaysUpdateSourcePose=true}
   Output Pose(Result=Slot 'AdditiveFullBody'(Source=Slot 'FullBody'(Source=@n1) {bAlwaysUpdateSourcePose=true, slotName=FullBody}) {slotName=AdditiveFullBody})
  ]}
 Base SM {
  entry=Rest
  path=AnimGraph.AnimGraphNode_StateMachine_0.Base SM
  ids=id
  nodes [
   AnimStateNode_0 = Rest() {
    is=title AnimStateNode
    stateEntered {guid=#1}
    stateFullyBlended {guid=#2}
    stateLeft {guid=#3}
   }
   AnimStateNode_2 = Combat Idle() {
    is=title AnimStateNode
    stateEntered {guid=#4}
    stateFullyBlended {guid=#5}
    stateLeft {guid=#6}
   }
   AnimStateTransitionNode_2 = AnimStateTransitionNode()
   AnimStateTransitionNode_4 = AnimStateTransitionNode()
  ]
  links [
   {
    dst=Rest
    rule=AnimGraph.AnimGraphNode_StateMachine_0.Base SM.AnimStateTransitionNode_2.Transition
    src=Combat Idle
    via=AnimStateTransitionNode_2
   }
   {
    dst=Combat Idle
    rule=AnimGraph.AnimGraphNode_StateMachine_0.Base SM.AnimStateTransitionNode_4.Transition
    src=Rest
    via=AnimStateTransitionNode_4
   }
  ]
 }
 Rest {
  path=AnimGraph.AnimGraphNode_StateMachine_0.Base SM.AnimStateNode_0.Rest
  nodes [Output Animation Pose(Result=Sequence Player 'MaskMaker_Idle_Rest'() {sequence=&/MaskMaker_Idle_Rest})]
 }
 Combat Idle {
  path=AnimGraph.AnimGraphNode_StateMachine_0.Base SM.AnimStateNode_2.Combat Idle
  nodes [Output Animation Pose(Result=Sequence Player 'MaskMaker_Idle_Combat'() {sequence=&/MaskMaker_Idle_Combat})]
 }
 Transition {
  path=AnimGraph.AnimGraphNode_StateMachine_0.Base SM.AnimStateTransitionNode_2.Transition
  nodes [Result(bCanEnterTransition=Get bIsUnaware() -> (bIsUnaware:bool))]
 }
 Transition {
  path=AnimGraph.AnimGraphNode_StateMachine_0.Base SM.AnimStateTransitionNode_4.Transition
  nodes [Result(bCanEnterTransition=Get bEnterCombatIdle() -> (bEnterCombatIdle:bool))]
 }
 EventGraph {links [], nodes []}
]
variables []
