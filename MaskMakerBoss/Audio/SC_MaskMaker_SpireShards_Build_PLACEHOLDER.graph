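# Placeholder sound cue for the build-up of the Mask Maker boss's Spire Shards attack, classed as boss-fight SFX.
# A Random node picks one of four equally weighted waves and feeds a Modulator for slight pitch and volume variation.
# It uses emitter attenuation, keeps playing when silent and caps concurrency, stopping the farthest instance first.
# Distance-scaled sends feed music and weapon sidechain buses, strongest up close.
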SoundCue /MaskMakerBoss/Audio/SC_MaskMaker_SpireShards_Build_PLACEHOLDER
root=Modulator_1
nodes [
 {id=Modulator_1, children [Random_0]}
 {id=Random_0, children [WavePlayer_4,WavePlayer_5,WavePlayer_6,WavePlayer_7], properties {weights [1,1,1,1]}}
 {id=WavePlayer_4, properties {soundWaveAssetPtr=&/MaskMaker_SpireShards_Build_04_PLACEHOLDER}}
 {id=WavePlayer_5, properties {soundWaveAssetPtr=&/MaskMaker_SpireShards_Build_03_PLACEHOLDER}}
 {id=WavePlayer_6, properties {soundWaveAssetPtr=&/MaskMaker_SpireShards_Build_02_PLACEHOLDER}}
 {id=WavePlayer_7, properties {soundWaveAssetPtr=&/MaskMaker_SpireShards_Build_01_PLACEHOLDER}}
]
properties {
 attenuationSettings=&/SA_MaskMaker_SpireShards_Emitter_PLACEHOLDER
 bOverrideConcurrency=true
 maxDistance=5500
 soundClassObject=&Fort_BossFight_SFX
 virtualizationMode=PlayWhenSilent
 volumeMultiplier=1
 busSends [
  {
   audioBus=&AB_Sidechain_FortMusic_Vol
   maxSendDistance=2500
   maxSendLevel=0
   minSendDistance=0
   minSendLevel=1
   sendLevel=0
   soundSourceBus=null
   sourceBusSendLevelControlMethod=Linear
  }
  {
   audioBus=&AB_Sidechain_FortWeapon_Vol
   maxSendDistance=2500
   maxSendLevel=0
   minSendDistance=0
   minSendLevel=1
   sendLevel=0
   soundSourceBus=null
   sourceBusSendLevelControlMethod=Linear
  }
 ]
 concurrencyOverrides {maxCount=4, resolutionRule=StopFarthestThenPreventNew}
}
