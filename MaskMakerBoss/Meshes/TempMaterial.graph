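# Temporary masked, lit material in the MaskMakerBoss meshes folder, with parameters for a hit-style flash and noise-based erosion.
# A Flash scalar blends between two orange tones and also fades in a tinted, intensity-scaled Fresnel rim.
# An Erode scalar minus tiled low-res blurred noise forms a dissolve-style mask value.
# The graph shows no links to material outputs, and a DebugTimeSine call and a constant feed nothing.
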Material /MaskMakerBoss/Meshes/TempMaterial
graph {
 properties {bAutomaticallySetUsageInEditor=false, blendMode=BLEND_Masked, previewMesh=&/Mask1, usedShadingModels=Default Lit}
 nodes [
  Constant()
  n1 = VectorParameter() {defaultValue=<rgba 1 0.0489 0.0933 1>}
  n2 = Fresnel(ExponentIn=ScalarParameter() {defaultValue=1}, BaseReflectFractionIn=ScalarParameter() {parameterName=Param_1})
  n3 = ScalarParameter() {parameterName=Flash}
  LinearInterpolate(B=Multiply(A=Multiply(A=@n1.RGB, B=@n2), B=ScalarParameter() {defaultValue=2, parameterName=Param_2}), Alpha=@n3)
  LinearInterpolate(A=Constant3Vector() {constant=<rgba 1 0.3829 0.0307 1>}, B=Constant3Vector() {constant=<rgba 0.5365 0.203 0.0025 1>}, Alpha=@n3)
  MaterialFunctionCall() {materialFunction=&DebugTimeSine}
  n4 = TextureCoordinate()
  n5 = ScalarParameter() {defaultValue=3, parameterName=Param_3}
  Subtract(
   A=ScalarParameter() {defaultValue=1, parameterName=Erode}
   B=TextureSample(UVs=Multiply(A=@n4, B=@n5)) {samplerType=SAMPLERTYPE_Grayscale, texture=&LowResBlurredNoise}.R
  )
 ]
}
