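# Niagara system for the MaskMaker boss's beam attack: a pink energy beam with muzzle and impact effects.
# Layered ribbons link the system origin to a computed end point, with charge meshes, rings and light at the source.
# User Stop Location places leaks, flare, streaks, colliding GPU sparks and an optional normal-aligned scorch decal.
# User width and opacity scale the layers, while time-of-day and LOD remap values temper brightness and impact size.
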NiagaraSystem /MaskMakerBoss/VFX/NS_MaskMaker_BeamAttack
dependencies {
 usedDataInterfaces [
  &NiagaraDataInterfaceCurve
  &NiagaraDataInterfaceCamera
  &NiagaraDataInterfaceVector2DCurve
  &NiagaraDataInterfaceSkeletalMesh
 ]
 usedDynamicInputs [
  &Multiply_VectorByFloat
  &FN_LOD_Distance_Remap
  &VectorLength
  &Subtract_Position
  &NormalizeVector
  &SubtractVectorFromPosition
  &RemapRange
 ]
 usedModules [
  &self:SystemSpawnScript.SystemScriptSource.SystemScriptGraph.NiagaraNodeAssignment_1.SetVariables_#1
  &SystemState
  &InterpolateOverTime
  &self:SystemSpawnScript.SystemScriptSource.SystemScriptGraph.NiagaraNodeAssignment_5.SetVariables_#2
  &self:SystemSpawnScript.SystemScriptSource.SystemScriptGraph.NiagaraNodeAssignment_3.SetVariables_#3
  &self:SystemSpawnScript.SystemScriptSource.SystemScriptGraph.NiagaraNodeAssignment_0.SetVariables_#4
  &self:SystemSpawnScript.SystemScriptSource.SystemScriptGraph.NiagaraNodeAssignment_6.SetVariables_#5
  &self:SystemSpawnScript.SystemScriptSource.SystemScriptGraph.NiagaraNodeAssignment_2.SetVariables_#6
  &self:SystemSpawnScript.SystemScriptSource.SystemScriptGraph.NiagaraNodeAssignment_4.SetVariables_#7
 ]
}
emitters {
 EnergyBeam {
  emitterUpdateScript [
   &EmitterState()
   &SpawnBurst_Instantaneous(Spawn Count=6, Use Spawn Probability=false, UseLoopCountLimit=false)
   &SampleSkeletalMesh(
    AdditionalYawPitchRoll=false
    Mesh Sampling Type:ENiagaraSkelSamplingModeFull=::NewEnumerator5
    Skeletal Mesh:NiagaraDataInterfaceSkeletalMesh=$System.SK_Mesh
    Socket Sampling Mode:ENiagaraSocketSamplingMode=::NewEnumerator5
   )
   /Niagara/Modules/Beams/BeamEmitterSetup(
    Absolute Beam End=true
    Beam End:NiagaraPosition=/Niagara/DynamicInputs/Helpers/SimulationPosition()
    Beam Start:NiagaraPosition=$System.BeamEndLocation
   )
  ]
  particleSpawnScript [
   &InitializeParticle(Lifetime=0.65, Mesh Scale Mode=::NewEnumerator3, Mesh Uniform Scale=50, Position Mode=::NewEnumerator0, Write Material Random=true)
   /Niagara/Modules/Beams/SpawnBeam()
   /Niagara/Modules/Beams/BeamWidth(Beam Width:NiagaraFloat=&Multiply_Float(
     A=&Multiply_Float(
      A=&Lerp_Float(
       A=&Multiply_Float(
        A=20
        B=&Lerp_Float(A=1, B=&MakeFloatFromVector(Vector=$Engine.Owner.Scale, Channel=::NewEnumerator0), Alpha=1)
       )
       B=&Multiply_Float(
        A=100
        B=&Lerp_Float(A=1, B=&MakeFloatFromVector(Vector=$Engine.Owner.Scale, Channel=::NewEnumerator0), Alpha=0.3)
       )
       Alpha=&FloatFromCurve(
        FloatCurve=DataInterface() {propertyValues {Curve {keys [{tangentMode=RCTM_User, time=0, value=1}, {tangentMode=RCTM_User, time=1, value=0}]}}}
        CurveIndex=$Particles.RibbonLinkOrder
       )
      )
      B=&Lerp_Float(A=0, B=1, Alpha=$System.BeamLengthScalar)
     )
     B=$User.BeamWidth
    ))
   /SetVariables(
    Particles.BeamWidthBlack:NiagaraFloat=&Multiply_Float(A=&Multiply_Float(A=$Particles.RibbonWidth, B=$User.BeamWidth), B=1.1)
    Particles.BeamWidthBlack02:NiagaraFloat=&Multiply_Float(A=&Multiply_Float(A=$Particles.RibbonWidth, B=$User.BeamWidth), B=0.9)
    Particles.BlackEnergyColor:LinearColor=$Particles.Color
   ) {isSetParameters=true}
   /SetVariables(Particles.BeamWidthLightning:NiagaraFloat=&Multiply_Float(A=$Particles.RibbonWidth, B=&Multiply_Float(A=2.8, B=$User.BeamWidth))) {isSetParameters=true}
   /SetVariables(
    Particles.LightningBeamAxis01:Vector3f=<1 0 1>
    Particles.LightningBeamAxis02:Vector3f=<1 0 -1>
    Particles.LightningBeamAxis03:Vector3f=<1 0 0>
   ) {isSetParameters=true}
   &InitialMeshOrientation(Mesh Orientation Mode=::NewEnumerator2, Orientation Vector=$System.BeamDirection, Use Rotation Vector=false)
  ]
  particleUpdateScript [
   &ParticleState()
   /Niagara/Modules/Beams/UpdateBeam()
   &DynamicMaterialParameters(
    Index 0 Param 1=&Lerp_Float(A=1, B=-0.075, Alpha=$System.BeamLengthScalar)
    Index 0 Param 4=&FloatFromCurve(
     FloatCurve=DataInterface() {propertyValues {Curve {keys [{arriveTangent=4.1862, interpMode=RCIM_Cubic, leaveTangent=4.1862, tangentMode=RCTM_User, time=0, value=0}, <at 0.5 1 RCIM_Cubic>]}}}
     CurveIndex=$Emitter.Age
    )
    Index 1 Param 1=&MakeFloatFromVector(Vector=$Emitter.BeamDirection, Channel=::NewEnumerator0)
    Index 1 Param 2=&MakeFloatFromVector(Vector=$Emitter.BeamDirection, Channel=::NewEnumerator1)
    Index 1 Param 3=&MakeFloatFromVector(Vector=$Emitter.BeamDirection, Channel=::NewEnumerator2)
    Module.Index0Param0Write=true
    Module.Index0Param1Write=true
    Module.Index0Param2Write=true
    Module.Index0Param3Write=true
    Module.Index1Param0Write=true
    Module.Index1Param1Write=true
    Module.Index1Param2Write=true
    Module.Index1Param3Write=true
    Write Parameter Index 1=true
   )
   &SolveForcesAndVelocity(Clamp Velocity=false, Limit Acceleration=false)
   &ScaleColor(
    Scale Alpha=&Multiply_Float(
     A=&FNTimeOfDayScalar(Night Scale=0.15, Morning Scale=0.7, Evening Scale=0.7)
     B=&Lerp_Float(A=0.6, B=1, Alpha=$System.ImpactLODRemap)
    )
    Scale RGB=&VectorFromFloat(Value=&Lerp_Float(
      A=1
      B=2
      Alpha=&RandomRangeFloat(
       Minimum=&FNTimeOfDayScalar(Night Scale=1, Morning Scale=0.6, Day Scale=0.6, Evening Scale=0.6)
       Maximum=1
       Evaluation Type=::NewEnumerator1
      )
     ))
    ScaleA=true
    ScaleRGB=true
   )
   &ScaleColor(
    Scale Alpha=&FloatFromCurve(
     FloatCurve=DataInterface() {propertyValues {Curve {keys [<at 0 1 RCIM_Cubic>, {arriveTangent=-1, interpMode=RCIM_Cubic, leaveTangent=-1, time=0.5, value=0.25}, <at 1 0 RCIM_Cubic>]}}}
     Scale Curve=$User.OverallOpacity
    )
    Scale RGB=&VectorFromFloat(Value=&Multiply_Float(A=$User.OverallOpacity, B=&Lerp_Float(A=1, B=20, Alpha=$System.BeamShootoutFlash)))
    ScaleA=true
    ScaleRGB=true
   ) {module=ScaleColor001}
   /SetVariables(Particles.BlackEnergyColor:LinearColor=&MakeLinearColor(
     R:NiagaraFloat=1
     G:NiagaraFloat=1
     B:NiagaraFloat=1
     A:NiagaraFloat=&Multiply_Float(
      A=&Multiply_Float(
       A=&FNTimeOfDayScalar(Night Scale=0.5, Morning Scale=0.7, Evening Scale=0.7)
       B=&Lerp_Float(A=2, B=10, Alpha=$System.ImpactLODRemap)
      )
      B=&FloatFromCurve(
       FloatCurve=DataInterface() {propertyValues {Curve {
          defaultValue=1
          keys [{interpMode=RCIM_Cubic, tangentMode=RCTM_User, time=0.7, value=1}, {interpMode=RCIM_Cubic, tangentMode=RCTM_User, time=1, value=0}]
         }}}
       Scale Curve=$User.OverallOpacity
      )
     )
    )) {isSetParameters=true}
   /Niagara/Modules/Beams/ScaleBeamWidth(Beam Width Scale:NiagaraFloat=&Lerp_Float(A=1, B=1.1, Alpha=$System.BeamShootoutFlash))
  ]
  rendererClasses [&NiagaraRibbonRendererProperties]
  renderers [
   {rendererClass=&NiagaraRibbonRendererProperties}
   {rendererClass=&NiagaraRibbonRendererProperties, rendererIndex=1}
   {rendererClass=&NiagaraRibbonRendererProperties, rendererIndex=2}
   {rendererClass=&NiagaraRibbonRendererProperties, rendererIndex=3}
   {rendererClass=&NiagaraRibbonRendererProperties, rendererIndex=4}
   {rendererClass=&NiagaraRibbonRendererProperties, rendererIndex=5}
  ]
 }
 EnergyChargeMesh {
  emitterUpdateScript [
   &EmitterState(Life Cycle Mode=::NewEnumerator1, Loop Behavior=::NewEnumerator1, Loop Duration=0.7, UseLoopDelay=true)
   &SpawnRate(SpawnRate=7, Use Spawn Probability=false)
  ]
  particleSpawnScript [
   &InitializeParticle(
    Color=<rgba 1 0.0418 0.3332 1>
    Lifetime=0.3
    Mesh Scale=&Multiply_Vector(A=<1.5 1 1>, B=$Engine.Owner.Scale)
    Mesh Scale Mode=::NewEnumerator4
    UsePositionOffset=false
    Write Material Random=true
   )
   &SystemLocation(
    MaskSpawnGroup=false
    Offset=&TransformVector(
     Vector=&Multiply_Vector(
      A=$System.BeamDirection
      B=&VectorFromFloat(Value=&Multiply_Float(A=$User.BeamWidth, B=200))
     )
     Source Space=::World
     Destination Space=::Local
    )
   )
  ]
  particleUpdateScript [
   &ParticleState()
   &UpdateMeshOrientation(
    Facing Direction=$System.BeamDirection
    Facing Interpolation=false
    Orientation Method=::NewEnumerator1
    Override Quat=false
    TransformFacing=false
    TransformUp=false
   ) {module=UpdateMeshOrientation001}
   &InheritSourceMovement(ApplyForce=false, ApplyPosition=true, ApplyPositionFalloff=false)
   &ScaleMeshSize(Scale Factor=&Multiply_Vector(
     A=&RandomRangeVector(Minimum=<0.8 0.7 0.7>, Maximum=<0.8 1 1>)
     B=&VectorFromFloat(Value=$User.BeamWidth)
    ))
   &ScaleColor(
    Scale Alpha=&FloatFromCurve(
     FloatCurve=DataInterface() {propertyValues {Curve {keys [
         {interpMode=RCIM_Cubic, tangentMode=RCTM_User, time=0, value=1}
         {arriveTangent=-1, interpMode=RCIM_Cubic, leaveTangent=-1, time=0.4981, value=0.1966}
         {interpMode=RCIM_Cubic, tangentMode=RCTM_User, time=1, value=0}
        ]}}}
     Scale Curve=&Multiply_Float(
      A=$User.OverallOpacity
      B=&FloatFromCurve(
       FloatCurve=DataInterface() {propertyValues {Curve {keys [
           {arriveTangent=-1.9515, interpMode=RCIM_Cubic, leaveTangent=-1.9515, tangentMode=RCTM_User, time=0, value=1}
           {arriveTangent=-1, interpMode=RCIM_Cubic, leaveTangent=-1, time=0.4037, value=0.1578}
           <at 1 0 RCIM_Cubic>
          ]}}}
       CurveIndex=$System.Age
      )
     )
    )
    Scale RGB=&VectorFromFloat(Value=&FloatFromCurve(
      FloatCurve=DataInterface() {propertyValues {Curve {keys [
          {arriveTangent=9.0725, interpMode=RCIM_Cubic, leaveTangent=9.0725, tangentMode=RCTM_User, time=0, value=0}
          {arriveTangent=-2.5249, interpMode=RCIM_Cubic, leaveTangent=-2.5249, tangentMode=RCTM_User, time=0.1408, value=0.9847}
          {arriveTangent=-0.0405, interpMode=RCIM_Cubic, leaveTangent=-0.0405, tangentMode=RCTM_User, time=1, value=0}
         ]}}}
      Scale Curve=100
     ))
    ScaleA=true
    ScaleRGB=true
   )
   &SolveForcesAndVelocity(Clamp Velocity=false, Limit Acceleration=false)
   &DynamicMaterialParameters(
    Index 0 Param 1=&FloatFromCurve(FloatCurve=DataInterface() {propertyValues {Curve {keys [{tangentMode=RCTM_User, time=0, value=0}, {tangentMode=RCTM_User, time=1, value=1}]}}})
    Module.Index0Param0Write=true
    Module.Index0Param1Write=true
    Module.Index0Param2Write=true
    Module.Index0Param3Write=true
   )
  ]
  rendererClasses [&NiagaraMeshRendererProperties]
  renderers [{rendererClass=&NiagaraMeshRendererProperties, rendererIndex=0}]
 }
 EnergyLeak {
  emitterUpdateScript [
   &EmitterState(Life Cycle Mode=::NewEnumerator1, Loop Behavior=::NewEnumerator1, Loop Duration=0.6, UseLoopDelay=false)
   &SpawnBurst_Instantaneous(Use Spawn Probability=false, UseLoopCountLimit=false)
  ]
  particleSpawnScript [
   &InitializeParticle(
    Color=<rgba 8 0.6524 2.2972 1>
    Lifetime=0.5
    Mesh Scale Mode=::NewEnumerator3
    Mesh Uniform Scale=&Multiply_Float(A=10, B=$User.BeamWidth)
    Position=$User.Stop Location
    Position Mode=::NewEnumerator1
    UsePositionOffset=false
    Write Material Random=true
   )
   &InitialMeshOrientation(Mesh Axis To Orient=<0 0 1>, Mesh Orientation Mode=::NewEnumerator2, Use Rotation Vector=false)
  ]
  particleUpdateScript [
   &ParticleState()
   &ScaleColor(
    Scale Alpha=&FloatFromCurve(
     FloatCurve=DataInterface() {propertyValues {Curve {keys [
         {arriveTangent=9.0725, interpMode=RCIM_Cubic, leaveTangent=9.0725, tangentMode=RCTM_User, time=0, value=0}
         {arriveTangent=-0.4058, interpMode=RCIM_Cubic, leaveTangent=-0.4058, tangentMode=RCTM_User, time=0.1, value=1}
         {arriveTangent=-1.1111, interpMode=RCIM_Cubic, leaveTangent=-1.1111, time=0.8, value=0.9}
         {arriveTangent=0.1594, interpMode=RCIM_Cubic, leaveTangent=0.1594, tangentMode=RCTM_User, time=1, value=0}
        ]}}}
     Scale Curve=&Multiply_Float(
      A=&Multiply_Float(
       A=&FNTimeOfDayScalar(Night Scale=0.2, Morning Scale=1, Evening Scale=1)
       B=&Lerp_Float(A=0.05, B=1, Alpha=$System.ImpactLODRemap)
      )
      B=$User.OverallOpacity
     )
    )
    Scale RGB=&VectorFromFloat(Value=&Multiply_Float(
      A=&FloatFromCurve(FloatCurve=DataInterface() {propertyValues {Curve {
          defaultValue=0
          keys [
           {arriveTangent=9.0725, interpMode=RCIM_Cubic, leaveTangent=9.0725, tangentMode=RCTM_User, time=0, value=1}
           {arriveTangent=-2.5249, interpMode=RCIM_Cubic, leaveTangent=-2.5249, tangentMode=RCTM_User, time=0.1408, value=0.9847}
           {arriveTangent=-1.146, interpMode=RCIM_Cubic, leaveTangent=-1.146, time=0.6369, value=0.6402}
           {arriveTangent=-0.0405, interpMode=RCIM_Cubic, leaveTangent=-0.0405, tangentMode=RCTM_User, time=1, value=0}
          ]
         }}})
      B=5
     ))
    ScaleA=true
    ScaleRGB=true
   ) {module=ScaleColor002}
   &ScaleMeshSize(Scale Factor=&VectorFromFloat(Value=1)) {module=ScaleMeshSize001}
   &ScaleMeshSize(Scale Factor=&VectorFromFloat(Value=&Lerp_Float(
      A=0.5
      B=1
      Alpha=&FloatFromCurve(FloatCurve=DataInterface() {propertyValues {Curve {keys [
           {arriveTangent=2.748, interpMode=RCIM_Cubic, leaveTangent=2.748, tangentMode=RCTM_User, time=0, value=0}
           {arriveTangent=0.2856, interpMode=RCIM_Cubic, leaveTangent=0.2856, tangentMode=RCTM_User, time=1, value=1}
          ]}}})
     )))
  ]
  rendererClasses [&NiagaraMeshRendererProperties]
  renderers [{rendererClass=&NiagaraMeshRendererProperties, rendererIndex=0}]
 }
 EnergyLeak001 {
  emitterUpdateScript [
   &EmitterState(Life Cycle Mode=::NewEnumerator1, Loop Behavior=::NewEnumerator1, Loop Duration=1, UseLoopDelay=false)
   &SpawnBurst_Instantaneous(Use Spawn Probability=false, UseLoopCountLimit=false)
  ]
  particleSpawnScript [
   &InitializeParticle(
    Color=<rgba 0.035 0.0026 0.0149 1>
    Lifetime=0.5
    Mesh Scale Mode=::NewEnumerator3
    Mesh Uniform Scale=&Multiply_Float(A=20, B=$User.BeamWidth)
    Position=$User.Stop Location
    Position Mode=::NewEnumerator1
    Sprite Size Mode=::NewEnumerator3
    Uniform Sprite Size=&Multiply_Float(A=1248.2896, B=$User.BeamWidth)
    UsePositionOffset=false
    Write Material Random=true
   )
   &InitialMeshOrientation(Mesh Axis To Orient=<0 0 1>, Mesh Orientation Mode=::NewEnumerator2, Use Rotation Vector=false)
  ]
  particleUpdateScript [
   &ParticleState()
   &ScaleSpriteSize(
    Uniform Curve Scale=0.7
    Uniform Curve Sprite Scale=DataInterface() {propertyValues {Curve {keys [{arriveTangent=0.6429, interpMode=RCIM_Cubic, leaveTangent=0.6429, tangentMode=RCTM_User, time=0, value=0.5}, <at 1 1 RCIM_Cubic>]}}}
   )
   &Timeline(
    Play Rate=&FloatFromCurve(FloatCurve=DataInterface() {propertyValues {Curve {keys [<at 0 2>,<at 1 1.3>]}}})
    Timeline End=1000
   )
   &CameraOffset(Camera Offset Amount=&Multiply_Float(
     A=-500
     B=&FloatFromCurve(
      FloatCurve=DataInterface() {propertyValues {Curve {keys [<at 0 1>,<at 1 0>]}}}
      CurveIndex=/Game/Effects/Niagara/Modules/Math/FN_CameraDistance(
       0Dist:NiagaraFloat=500
       1Dist:NiagaraFloat=2000
       Camera Query:NiagaraDataInterfaceCamera=DataInterface()
       Position:NiagaraPosition=$Particles.Position
      )
     )
    )) {enabled=false}
   &ScaleColor(
    Scale Alpha=&FloatFromCurve(
     FloatCurve=DataInterface() {propertyValues {Curve {keys [
         {arriveTangent=0.1216, interpMode=RCIM_Cubic, leaveTangent=0.1216, tangentMode=RCTM_User, time=0.07, value=0}
         <at 0.3 1 RCIM_Cubic>
         {arriveTangent=-0.0967, interpMode=RCIM_Cubic, leaveTangent=-0.0967, tangentMode=RCTM_User, time=1, value=0}
        ]}}}
     Scale Curve=&Multiply_Float(
      A=&Multiply_Float(A=3, B=$User.OverallOpacity)
      B=&Multiply_Float(
       A=&FNTimeOfDayScalar(Night Scale=0.2, Morning Scale=1, Evening Scale=1)
       B=&Lerp_Float(A=0.05, B=1, Alpha=$System.ImpactLODRemap)
      )
     )
    )
    ScaleA=true
    ScaleRGB=true
   ) {module=ScaleColor002}
   &DynamicMaterialParameters(
    Index 0 Param 1=$Particles.Timeline.Time
    Module.Index0Param0Write=true
    Module.Index0Param1Write=true
    Module.Index0Param2Write=true
    Module.Index0Param3Write=true
   )
  ]
  rendererClasses [&NiagaraSpriteRendererProperties]
  renderers [{rendererClass=&NiagaraSpriteRendererProperties}]
 }
 ImpactLight {
  emitterSpawnScript [/SetVariables(Emitter.CameraOffsetSprite=50) {isSetParameters=true, module=SetVariables_#8}]
  emitterUpdateScript [
   &EmitterState()
   &SpawnBurst_Instantaneous(Use Spawn Probability=true, UseLoopCountLimit=true)
   &SpawnBurst_Instantaneous(Use Spawn Probability=false, UseLoopCountLimit=false) {module=SpawnBurst_Instantaneous001}
  ]
  particleSpawnScript [
   &InitializeParticle(
    Lifetime=0.6
    Position=$User.Stop Location
    Position Mode=::NewEnumerator1
    Sprite Rotation Mode=::NewEnumerator2
    Sprite Size Mode=::NewEnumerator3
    Uniform Sprite Size=400
    UsePositionOffset=false
    Write Material Random=true
   )
   &DynamicMaterialParameters(Module.Index0Param0Write=false, Module.Index0Param1Write=false, Module.Index0Param2Write=false, Module.Index0Param3Write=false)
  ]
  particleUpdateScript [
   &ParticleState()
   &ScaleColor(
    Scale Alpha=&Multiply_Float(
     A=&MakeCustomFloatFromBool(
      Bool=&SetBoolByFloatComparison(A=$System.BeamLength, B=9500, Comparison Type=::NewEnumerator6)
      True Float=0.5
      False Float=0.1
     )
     B=&Multiply_Float(
      A=&FNTimeOfDayScalar(Night Scale=0.2, Morning Scale=0.5, Evening Scale=0.5)
      B=&Lerp_Float(A=0.15, B=1, Alpha=$System.ImpactLODRemap)
     )
    )
    Scale RGB=<1 0.05 0.3>
    ScaleA=true
    ScaleRGB=true
   ) {module=ScaleColor002}
   &ScaleColor(
    Scale Alpha=&FloatFromCurve(
     FloatCurve=DataInterface() {propertyValues {Curve {
        defaultValue=1
        keys [{interpMode=RCIM_Cubic, tangentMode=RCTM_User, time=0, value=1}, {interpMode=RCIM_Cubic, tangentMode=RCTM_User, time=1, value=0}]
       }}}
     Scale Curve=$User.OverallOpacity
    )
    ScaleA=true
    ScaleRGB=true
   ) {module=ScaleColor001}
   /SetVariables(Particles.FlareColor=&MultiplyLinearColor(
     Linear Color=$Particles.Color
     Scale Factor=&MakeVector4(
      X=1
      Y=1
      Z=1
      W=&Multiply_Float(
       A=&Multiply_Float(
        A=&MakeCustomFloatFromBool(
         Bool=&SetBoolByFloatComparison(A=$System.BeamLength, B=9500, Comparison Type=::NewEnumerator6)
         True Float=1
         False Float=0.05
        )
        B=&Lerp_Float(A=0.3, B=1, Alpha=$System.ImpactLODRemap)
       )
       B=&FloatFromCurve(
        FloatCurve=DataInterface() {propertyValues {Curve {keys [<at 0 0.5>,<at 1 1>]}}}
        CurveIndex=&Divide_Float(
         A=&DistanceBetweenPositions(
          A=&ReturnCameraPosition(Coordinate Space:ENiagaraCoordinateSpace=::Simulation, Camera Query:NiagaraDataInterfaceCamera=DataInterface())
          B=$User.Stop Location
         )
         B=&Multiply_Float(A=300, B=&MakeFloatFromVector(Vector=$Engine.Owner.Scale, Channel=::NewEnumerator0))
        )
        Scale Curve=&FNTimeOfDayScalar(Night Scale=0.3, Morning Scale=0.7, Evening Scale=0.7)
       )
      )
     )
    )) {isSetParameters=true}
   &ApplyOwnerScaleToAttributes(Scale Initial Sprite Size=true, Sprite Size Scale Amount=0.2)
   &SpriteRotationRate(Rotation Rate=1000)
   &ScaleSpriteSize(
    Scale Sprite Size Mode=::NewEnumerator0
    Uniform Scale Factor=&Lerp_Float(
     A=0.5
     B=1
     Alpha=&RandomRangeFloat(
      Minimum=&FNTimeOfDayScalar(Night Scale=1, Morning Scale=0.7, Day Scale=0.6, Evening Scale=0.7)
      Maximum=1
      Evaluation Type=::NewEnumerator1
     )
    )
   ) {module=ScaleSpriteSize001}
   &ScaleSpriteSize(
    Scale Sprite Size Mode=::NewEnumerator0
    Uniform Scale Factor=&Multiply_Float(
     A=&Multiply_Float(A=$User.BeamWidth, B=3)
     B=&Lerp_Float(A=0.5, B=1, Alpha=$System.ImpactLODRemap)
    )
   )
   &FN_Light_Attributes(
    Color=&MultiplyLinearColor(
     Linear Color=$Particles.Color
     Scale Factor=&MakeVector4_FromFloat(Float=&FNTimeOfDayScalar(Night Scale=0.5, Morning Scale=1.5, Day Scale=2, Evening Scale=1.5))
    )
    Create Color Attribute=true
    Create Position Attribute=true
    Create Radius Attribute=true
    Enable Particle Light Enable=true
    Light Enabled=true
    Position=&AddVectorToPosition(
     Position=$Particles.Position
     Vector=&TransformVector(
      Vector=&Multiply_Vector(
       A=Linked() {linkedVariable {name=System.BeamDirection, type=NiagaraPosition}}
       B=&VectorFromFloat(Value=0)
      )
      Source Space=::World
      Destination Space=::Local
     )
    )
    Radius=&Multiply_Float(
     A=&Multiply_Float(
      A=&Lerp_Float(A=400, B=500, Alpha=&RandomRangeFloat(Minimum=0.5, Maximum=1, Evaluation Type=::NewEnumerator1))
      B=&Lerp_Float(A=0.2, B=1, Alpha=$System.ImpactLODRemap)
     )
     B=$User.BeamWidth
    )
    Scale Color=&FloatFromCurve(
     FloatCurve=DataInterface() {propertyValues {Curve {
        defaultValue=1
        keys [{interpMode=RCIM_Cubic, tangentMode=RCTM_User, time=0, value=1}, {interpMode=RCIM_Cubic, tangentMode=RCTM_User, time=1, value=0}]
       }}}
     Scale Curve=&Multiply_Float(A=20, B=$User.OverallOpacity)
    )
   )
   /SetVariables(Particles.Position=$System.BeamEndLocation) {isSetParameters=true}
   &ScaleColor(
    Scale Alpha=&FloatFromCurve(
     FloatCurve=DataInterface() {propertyValues {Curve {keys [<at 0 0.3>,<at 1 1>]}}}
     CurveIndex=&Divide_Float(
      A=&DistanceBetweenPositions(
       A=&ReturnCameraPosition(Coordinate Space:ENiagaraCoordinateSpace=::Simulation, Camera Query:NiagaraDataInterfaceCamera=DataInterface())
       B=$User.Stop Location
      )
      B=&Multiply_Float(A=600, B=&MakeFloatFromVector(Vector=$Engine.Owner.Scale, Channel=::NewEnumerator0))
     )
    )
    ScaleA=true
    ScaleRGB=true
   )
  ]
  rendererClasses [&NiagaraSpriteRendererProperties, &NiagaraLightRendererProperties]
  renderers [
   {rendererClass=&NiagaraSpriteRendererProperties}
   {rendererClass=&NiagaraLightRendererProperties, rendererIndex=1}
   {rendererClass=&NiagaraSpriteRendererProperties, rendererIndex=2}
  ]
 }
 Light {
  emitterSpawnScript [/SetVariables(Emitter.CameraOffsetSprite=50) {isSetParameters=true, module=SetVariables_#8}]
  emitterUpdateScript [
   &EmitterState(Life Cycle Mode=::NewEnumerator1, Loop Behavior=::NewEnumerator1, Loop Duration=1, UseLoopDelay=false)
   &SpawnBurst_Instantaneous(Use Spawn Probability=false, UseLoopCountLimit=true)
  ]
  particleSpawnScript [
   &InitializeParticle(
    Color=<rgba 1 0.0374 0.3523 1>
    Lifetime=1.5
    Sprite Rotation Mode=::NewEnumerator2
    Sprite Size Mode=::NewEnumerator3
    Uniform Sprite Size=&Multiply_Float(A=300, B=$User.BeamWidth)
    UsePositionOffset=false
    Write Material Random=true
   )
   &DynamicMaterialParameters(Module.Index0Param0Write=true, Module.Index0Param1Write=true, Module.Index0Param2Write=true, Module.Index0Param3Write=true)
  ]
  particleUpdateScript [
   &ParticleState(Kill Particles When Lifetime Has Elapsed=false, Let Infinitely Lived Particles Die When Emitter Deactivates=true)
   &ApplyOwnerScaleToAttributes(Scale Initial Sprite Size=true)
   &ScaleSpriteSize(
    Scale Sprite Size Mode=::NewEnumerator0
    Uniform Scale Factor=&Lerp_Float(A=0.5, B=1, Alpha=&RandomRangeFloat(Minimum=0.5, Maximum=1, Evaluation Type=::NewEnumerator1))
   ) {module=ScaleSpriteSize001}
   &ScaleSpriteSize(
    Scale Sprite Size Mode=::NewEnumerator0
    Uniform Scale Factor=&Multiply_Float(
     A=0.4
     B=&Lerp_Float(
      A=1
      B=1.5
      Alpha=&FloatFromCurve(FloatCurve=DataInterface() {propertyValues {Curve {keys [
           {arriveTangent=0.6, interpMode=RCIM_Cubic, time=0.1, value=1.05}
           {arriveTangent=-1.1667, interpMode=RCIM_Cubic, leaveTangent=-1.1667, time=0.5, value=0.3}
           {arriveTangent=-0.0405, interpMode=RCIM_Cubic, leaveTangent=-0.0405, tangentMode=RCTM_User, time=1, value=0}
          ]}}})
     )
    )
   )
   &FN_Light_Attributes(
    Color=&MultiplyLinearColor(
     Linear Color=<rgba 1 0.0554 0.264 1>
     Scale Factor=&MakeVector4_FromFloat(Float=&FloatFromCurve(
       FloatCurve=DataInterface() {propertyValues {Curve {
          defaultValue=1
          keys [{interpMode=RCIM_Cubic, tangentMode=RCTM_User, time=0, value=1}, {interpMode=RCIM_Cubic, tangentMode=RCTM_User, time=1, value=0}]
         }}}
       Scale Curve=$User.OverallOpacity
      ))
    )
    Create Color Attribute=true
    Create Position Attribute=true
    Create Radius Attribute=true
    Enable Particle Light Enable=true
    Light Enabled=true
    Position=&AddVectorToPosition(
     Position=$Particles.Position
     Vector=&TransformVector(
      Vector=&Multiply_Vector(
       A=Linked() {linkedVariable {name=System.BeamDirection, type=NiagaraPosition}}
       B=&VectorFromFloat(Value=300)
      )
      Source Space=::World
      Destination Space=::Local
     )
    )
    Radius=&Multiply_Float(
     A=&Multiply_Float(
      A=&Lerp_Float(
       A=200
       B=400
       Alpha=&Multiply_Float(
        A=&FNTimeOfDayScalar(Night Scale=0.2, Morning Scale=1, Evening Scale=1)
        B=&RandomRangeFloat(Minimum=0.5, Maximum=1, Evaluation Type=::NewEnumerator1)
       )
      )
      B=&MakeFloatFromVector(Vector=$Engine.Owner.Scale, Channel=::NewEnumerator0)
     )
     B=$User.BeamWidth
    )
    Scale Color=20
   )
   &SolveForcesAndVelocity(Clamp Velocity=false, Limit Acceleration=false)
   &ScaleColor(
    Scale Alpha=&Multiply_Float(A=0.7, B=$User.OverallOpacity)
    Scale RGB=&VectorFromFloat(Value=&Multiply_Float(
      A=1
      B=&Lerp_Float(
       A=0
       B=2
       Alpha=&FloatFromCurve(FloatCurve=DataInterface() {propertyValues {Curve {keys [
            {arriveTangent=0.4945, interpMode=RCIM_Cubic, leaveTangent=0.4945, tangentMode=RCTM_User, time=0, value=1}
            {arriveTangent=-1, interpMode=RCIM_Cubic, leaveTangent=-1, time=0.3, value=0.275}
            {arriveTangent=-0.0405, interpMode=RCIM_Cubic, leaveTangent=-0.0405, tangentMode=RCTM_User, time=1, value=0}
           ]}}})
      )
     ))
    ScaleA=true
    ScaleRGB=true
   )
  ]
  rendererClasses [&NiagaraSpriteRendererProperties, &NiagaraLightRendererProperties]
  renderers [
   {rendererClass=&NiagaraSpriteRendererProperties}
   {rendererClass=&NiagaraLightRendererProperties, rendererIndex=1}
  ]
 }
 ScorchDecal {
  emitterUpdateScript [
   &EmitterState(Life Cycle Mode=::NewEnumerator1, Loop Behavior=::NewEnumerator1, Loop Duration=0.6, UseLoopDelay=false)
   &SpawnBurst_Instantaneous(
    Spawn Count=1
    Spawn Probability=&MakeCustomFloatFromBool(Bool=$User.SpawnDecal, True Float=1, False Float=0)
    Spawn Time=0.3
    Use Spawn Probability=true
   )
  ]
  particleSpawnScript [
   &InitializeParticle(
    Lifetime=3
    Position=$User.Stop Location
    Position Mode=::NewEnumerator1
    Sprite Size Mode=::NewEnumerator3
    Uniform Sprite Size=600
    UsePositionOffset=false
    Write Material Random=true
   )
   /SetVariables(
    Particles.DecalFade=1
    Particles.DecalOrientation:Quat4f=&OrientMeshToVector(Look At Direction=$User.DecalNormal, Reference Axis Vector=<1 0 0>)
    Particles.DecalSize:Vector3f=&MakeVector(
     X=350
     Y=&Multiply_Float(A=$User.BeamWidth, B=600)
     Z=&Multiply_Float(A=$User.BeamWidth, B=600)
    )
   ) {isSetParameters=true}
   /SetVariables(Particles.Position=$User.Stop Location) {isSetParameters=true}
  ]
  particleUpdateScript [
   &ParticleState()
   /SetVariables(Particles.DecalFade:NiagaraFloat=&Lerp_Float(
     A=0
     B=1
     Alpha=&FloatFromCurve(FloatCurve=DataInterface() {propertyValues {Curve {keys [
          {arriveTangent=12.3499, interpMode=RCIM_Cubic, leaveTangent=12.3499, tangentMode=RCTM_User, time=0, value=0}
          {arriveTangent=1.125, leaveTangent=1.125, time=0.2, value=1}
          {arriveTangent=-1.25, interpMode=RCIM_Cubic, leaveTangent=-1.25, time=0.8, value=0.95}
          {arriveTangent=-0.0037, interpMode=RCIM_Cubic, leaveTangent=-0.0037, tangentMode=RCTM_User, time=1, value=0}
         ]}}})
    )) {isSetParameters=true}
  ]
  rendererClasses [&NiagaraDecalRendererProperties]
  renderers [{rendererClass=&NiagaraDecalRendererProperties, rendererIndex=0}]
 }
 Sparks {
  simTarget=GPUComputeSim
  emitterUpdateScript [
   &EmitterState(Life Cycle Mode=::NewEnumerator1, Loop Behavior=::NewEnumerator1, Loop Duration=0.6, UseLoopDelay=false)
   &SpawnRate(SpawnRate=250, Use Spawn Probability=false)
  ]
  particleSpawnScript [
   &InitializeParticle(
    Lifetime Min=0.5
    Lifetime Mode=::NewEnumerator1
    Position=$User.Stop Location
    Position Mode=::NewEnumerator1
    Sprite Size Max=<6 6>
    Sprite Size Min=<2 2>
    Sprite Size Mode=::NewEnumerator2
    UsePositionOffset=false
    Write Material Random=true
   )
   &ShapeLocation(
    Cylinder Height=&Multiply_Float(
     A=&RandomRangeFloat(Minimum=0.5, Maximum=1)
     B=&Multiply_Float(A=$User.BeamWidth, B=100)
    )
    Cylinder Radius=&Multiply_Float(
     A=&RandomRangeFloat(Minimum=0.5, Maximum=1)
     B=&Multiply_Float(A=150, B=$User.BeamWidth)
    )
    MaskSpawnGroup=false
    Non Uniform Scale=&Divide_Vector(A:Vector3f=<1 1 1>, B:Vector3f=$Engine.Owner.Scale)
    Offset=<0 0 50>
    Offset Mode=::NewEnumerator0
    Override Seed=false
    Rotation Axis=$System.BeamDirection
    Rotation Mode=::NewEnumerator1
    Shape Origin=$User.Stop Location
    Shape Primitive=::NewEnumerator1
    Surface Only=true
    Surface Only Band Thickness=1
    Use Endcaps In Surface Only Mode=false
    UseOwnerScale=true
   )
   &AddVelocity(
    Rotation Coordinate Space=::World
    Velocity=&MakeVector(
     X=0
     Y=0
     Z=&Multiply_Float(A=&RandomRangeFloat(Minimum=0.5, Maximum=1), B=1)
    )
    Velocity Speed Scale=1000
   )
   &AddVelocity(
    Velocity=&TransformVector(
     Vector=&Multiply_Vector(A=$System.BeamDirection, B=&VectorFromFloat(Value=-1000))
     Source Space=::World
     Destination Space=::Local
    )
    Velocity Speed Scale=&RandomRangeFloat(Minimum=0.5, Maximum=1)
   ) {module=AddVelocity003}
   &AddVelocity(
    Module.Use Falloff Distance=false
    Origin Offset=&TransformVector(
     Vector=&Multiply_Vector(A=$System.BeamDirection, B=&VectorFromFloat(Value=1))
     Source Space=::World
     Destination Space=::Local
    )
    Override Seed=false
    Velocity Mode=::NewEnumerator1
    Velocity Origin=$User.Stop Location
    Velocity Speed=&Multiply_Float(A=&RandomRangeFloat(Minimum=0.6, Maximum=1), B=1000)
   ) {module=AddVelocity004}
  ]
  particleUpdateScript [
   &ParticleState()
   &ApplyOwnerScaleToAttributes(
    Forces Scale Amount:Vector3f=<0.2 0.2 0.2>
    Initial Velocity Scale Amount:Vector3f=<0.2 0.2 0.2>
    Scale Forces=true
    Scale Initial Velocity=true
   )
   &ScaleSpriteSize(Scale Sprite Size Mode=::NewEnumerator0, Uniform Scale Factor=&Multiply_Float(A=$User.BeamWidth, B=3))
   &ScaleSpriteSize(
    Uniform Curve Scale=&Multiply_Float(A=1, B=&Lerp_Float(A=0.1, B=1, Alpha=$System.ImpactLODRemap))
    Uniform Curve Sprite Scale=DataInterface() {propertyValues {Curve {keys [
        {interpMode=RCIM_Cubic, leaveTangent=-1.4666, tangentMode=RCTM_Break, time=0, value=1}
        {arriveTangent=-1, interpMode=RCIM_Cubic, leaveTangent=-1, time=0.7, value=0.2}
        {interpMode=RCIM_Cubic, tangentMode=RCTM_Break, time=1, value=0}
       ]}}}
   ) {module=ScaleSpriteSize001}
   &GravityForce(Gravity=&MakeVector(X=0, Y=0, Z=&Multiply_Float(A=-980, B=5)))
   &CurlNoiseForce(Noise Frequency=10, Noise Strength=8000, Pan Noise=false)
   &Drag(Drag=&RandomRangeFloat(Minimum=1, Maximum=4), Use Linear Drag=true, Use Rotational Drag=false)
   &Collision(
    Age Colliding Particles=false
    Friction=1
    GPU Collision Type=::NewEnumerator2
    KillOnCollision=false
    Module.Control Rest State Via Surface Normal=false
    Module.EnableMaxCollisionCount=false
    Module.Randomize Collision Normal Vector=false
    Restitution=0.02
   )
   &SolveForcesAndVelocity(Clamp Velocity=false, Limit Acceleration=false)
   &ScaleSpriteSizeBySpeed(Max Scale Factor=<1 4>)
   &ScaleColor(
    Scale Alpha=&FloatFromCurve(
     FloatCurve=DataInterface() {propertyValues {Curve {
        defaultValue=1
        keys [
         {interpMode=RCIM_Cubic, tangentMode=RCTM_User, time=0, value=1}
         {arriveTangent=-1, interpMode=RCIM_Cubic, leaveTangent=-1, time=0.161, value=0.3341}
         {interpMode=RCIM_Cubic, tangentMode=RCTM_User, time=1, value=0}
        ]
       }}}
     Scale Curve=$User.OverallOpacity
    )
    Scale RGB=&VectorFromFloat(Value=&Lerp_Float(
      A=1
      B=10
      Alpha=&FloatFromCurve(FloatCurve=DataInterface() {propertyValues {Curve {
          defaultValue=1
          keys [{interpMode=RCIM_Cubic, tangentMode=RCTM_User, time=0, value=1}, {interpMode=RCIM_Cubic, tangentMode=RCTM_User, time=1, value=0}]
         }}})
     ))
    ScaleA=true
    ScaleRGB=true
   )
  ]
  rendererClasses [&NiagaraSpriteRendererProperties]
  renderers [{rendererClass=&NiagaraSpriteRendererProperties}]
 }
 SplineRing001 {
  emitterUpdateScript [
   &EmitterState(Life Cycle Mode=::NewEnumerator1, Loop Behavior=::NewEnumerator1, Loop Duration=0.7, UseLoopDelay=true)
   &SpawnRate(SpawnRate=10, Use Spawn Probability=false)
  ]
  particleSpawnScript [
   &InitializeParticle(
    Color=<rgba 7.0876 1.2316 10 1>
    Lifetime=0.35
    Mesh Scale Mode=::NewEnumerator3
    Mesh Uniform Scale=&Multiply_Float(A=0.3, B=$User.BeamWidth)
    UsePositionOffset=false
    Write Material Random=true
   )
   &InitialMeshOrientation(
    Mesh Axis To Orient=<0 0 1>
    Mesh Orientation Mode=::NewEnumerator2
    Rotation=&MakeVector(X=0, Y=0, Z=&RandomRangeFloat(Minimum=0, Maximum=1))
    Use Rotation Vector=true
   )
   &SystemLocation(
    MaskSpawnGroup=false
    Offset=&TransformVector(
     Vector=&Multiply_Vector(
      A=$System.BeamDirection
      B=&VectorFromFloat(Value=&Multiply_Float(A=$User.BeamWidth, B=10))
     )
     Source Space=::World
     Destination Space=::Local
    )
   )
  ]
  particleUpdateScript [
   &ParticleState()
   &UpdateMeshOrientation(
    Facing Direction=&TransformVector(Vector=$System.BeamDirection, Source Space=::World, Destination Space=::Local)
    Facing Interpolation=false
    Orientation Method=::NewEnumerator1
    Override Quat=false
    TransformFacing=false
    TransformUp=false
   )
   &ScaleMeshSize(Scale Factor=&VectorFromFloat(Value=&Lerp_Float(
      A=0
      B=1
      Alpha=&FloatFromCurve(
       FloatCurve=DataInterface() {propertyValues {Curve {keys [
           {arriveTangent=4.1785, interpMode=RCIM_Cubic, leaveTangent=4.1785, tangentMode=RCTM_User, time=0, value=0}
           {arriveTangent=1.1179, interpMode=RCIM_Cubic, leaveTangent=1.1179, tangentMode=RCTM_User, time=0.2042, value=0.6228}
           {arriveTangent=0.0224, interpMode=RCIM_Cubic, leaveTangent=0.0224, tangentMode=RCTM_User, time=1, value=1}
          ]}}}
       Scale Curve=2
      )
     )))
   &DynamicMaterialParameters(
    Index 0 Param 1=&FloatFromCurve(FloatCurve=DataInterface() {propertyValues {Curve {keys [<at 0 0 RCIM_Cubic>, {arriveTangent=3.0069, interpMode=RCIM_Cubic, leaveTangent=3.0069, tangentMode=RCTM_User, time=1, value=1}]}}})
    Index 0 Param 2=&Multiply_Float(
     A=&Multiply_Float(
      A=&Lerp_Float(
       A=1
       B=8
       Alpha=&FloatFromCurve(
        FloatCurve=DataInterface() {propertyValues {Curve {keys [
            {arriveTangent=-2.29, interpMode=RCIM_Cubic, leaveTangent=-2.29, tangentMode=RCTM_User, time=0, value=1}
            {interpMode=RCIM_Cubic, tangentMode=RCTM_User, time=1, value=0}
           ]}}}
        Scale Curve=2
       )
      )
      B=&MakeFloatFromVector(Vector=$Engine.Owner.Scale, Channel=::NewEnumerator0)
     )
     B=$User.BeamWidth
    )
    Index 0 Param 3=&Lerp_Float(
     A=0
     B=100
     Alpha=&FloatFromCurve(
      FloatCurve=DataInterface() {propertyValues {Curve {keys [
          {arriveTangent=4.5741, interpMode=RCIM_Cubic, leaveTangent=4.5741, leaveTangentWeight=0.7804, tangentMode=RCTM_User, time=0, value=0}
          {arriveTangentWeight=0.1667, interpMode=RCIM_Cubic, tangentMode=RCTM_User, time=0.5, value=1}
         ]}}}
      Scale Curve=&Multiply_Float(A=3, B=$User.BeamWidth)
     )
    )
    Module.Index0Param0Write=true
    Module.Index0Param1Write=true
    Module.Index0Param2Write=true
    Module.Index0Param3Write=true
   )
   &ScaleColor(
    Scale Alpha=&Multiply_Float(
     A=$User.OverallOpacity
     B=&FloatFromCurve(
      FloatCurve=DataInterface() {propertyValues {Curve {keys [
          {arriveTangent=-1.9515, interpMode=RCIM_Cubic, leaveTangent=-1.9515, tangentMode=RCTM_User, time=0, value=1}
          {arriveTangent=-1, interpMode=RCIM_Cubic, leaveTangent=-1, time=0.4, value=0.2}
          <at 1 0 RCIM_Cubic>
         ]}}}
      CurveIndex=$System.Age
     )
    )
    ScaleA=true
    ScaleRGB=true
   )
  ]
  rendererClasses [&NiagaraMeshRendererProperties]
  renderers [{rendererClass=&NiagaraMeshRendererProperties, rendererIndex=0}]
 }
 Streaks_Fat {
  emitterUpdateScript [
   &EmitterState(Life Cycle Mode=::NewEnumerator1, Loop Behavior=::NewEnumerator1, Loop Duration=0.4, UseLoopDelay=false)
   &SpawnRate(SpawnRate=50, Use Spawn Probability=false)
  ]
  particleSpawnScript [
   &InitializeParticle(
    Color=<rgba 5 0.0571 0.6796 1>
    Lifetime Max=0.3
    Lifetime Min=0.25
    Lifetime Mode=::NewEnumerator1
    Position=$User.Stop Location
    Position Mode=::NewEnumerator1
    Sprite Size=&UniformRangedVector2D(Minimum=<20 100>, Maximum=<160 350>)
    Sprite Size Mode=::NewEnumerator4
    UsePositionOffset=false
    Write Material Random=true
   ) {module=InitializeParticle001}
   &ShapeLocation(
    Cylinder Height=&Multiply_Float(A=15, B=$User.BeamWidth)
    Cylinder Radius=&Multiply_Float(A=&RandomRangeFloat(Minimum=30, Maximum=80), B=$User.BeamWidth)
    MaskSpawnGroup=false
    Non Uniform Scale=&Divide_Vector(A:Vector3f=<1 1 1>, B:Vector3f=$Engine.Owner.Scale)
    Offset Mode=::NewEnumerator0
    Override Seed=false
    Rotation Axis=$System.BeamDirection
    Rotation Mode=::NewEnumerator1
    Shape Origin=$User.Stop Location
    Shape Primitive=::NewEnumerator1
    Surface Only=false
    UseOwnerScale=true
   )
   &AddVelocity(
    Module.Use Falloff Distance=false
    Origin Offset=&TransformVector(
     Vector=&Multiply_Vector(A=$System.BeamDirection, B=&VectorFromFloat(Value=50))
     Source Space=::World
     Destination Space=::Local
    )
    Override Seed=false
    Velocity Mode=::NewEnumerator1
    Velocity Origin=$User.Stop Location
    Velocity Speed=&Multiply_Float(
     A=&RandomRangeFloat(Minimum=0.2, Maximum=1)
     B=&Multiply_Float(A=$User.BeamWidth, B=500)
    )
   ) {module=AddVelocity002}
  ]
  particleUpdateScript [
   &ParticleState()
   &ApplyOwnerScaleToAttributes(Scale Initial Sprite Size=true, Sprite Size Scale Amount=0.5)
   &ScaleSpriteSize(Scale Sprite Size Mode=::NewEnumerator0, Uniform Scale Factor=&Multiply_Float(A=0.8, B=$User.BeamWidth)) {module=ScaleSpriteSize001}
   &SolveForcesAndVelocity(Clamp Velocity=false, Limit Acceleration=false)
   &ScaleSpriteSize(Scale Factor=&Vector2DFromCurve(
     Vector2Curve=DataInterface() {propertyValues {
       XCurve {keys [<at 0 0.75 RCIM_Cubic>, {arriveTangent=-0.75, interpMode=RCIM_Cubic, leaveTangent=-0.75, time=0.2, value=1.625}, <at 1 0 RCIM_Cubic>]}
       YCurve {keys [<at 0 1.625 RCIM_Cubic>,<at 1 2 RCIM_Cubic>]}
      }}
     ScaleCurve=&Multiply_Vector2D_ByFloat(
      Float=&Multiply_Float(A=1, B=&Lerp_Float(A=0.5, B=1, Alpha=$System.ImpactLODRemap))
      Vector2D=<1 1>
     )
    ))
   &ScaleColor(
    Scale Alpha=&FloatFromCurve(
     FloatCurve=DataInterface() {propertyValues {Curve {keys [<at 0 1 RCIM_Cubic>,<at 1 0>]}}}
     Scale Curve=&Multiply_Float(
      A=&Multiply_Float(
       A=&FNTimeOfDayScalar(Night Scale=0.2, Morning Scale=1, Evening Scale=1)
       B=&Lerp_Float(A=0.05, B=1, Alpha=$System.ImpactLODRemap)
      )
      B=$User.OverallOpacity
     )
    )
    Scale RGB=&VectorFromFloat(Value=&FloatFromCurve(FloatCurve=DataInterface() {propertyValues {Curve {keys [<at 0 10 RCIM_Cubic>,<at 1 1 RCIM_Cubic>]}}}))
    ScaleA=true
    ScaleRGB=true
   )
   &DynamicMaterialParameters(
    Index 0 Param 2=1
    Index 0 Param 3=1
    Index 0 Param 4=&FloatFromCurve(FloatCurve=DataInterface() {propertyValues {Curve {keys [<at 0.3 1>,<at 1 0.25>]}}})
    Module.Param0WriteEnabled=true
    Module.Param1WriteEnabled=true
    Module.Param2WriteEnabled=true
    Module.Param3WriteEnabled=true
   )
   &ScaleColor(
    Scale Alpha=&FloatFromCurve(
     FloatCurve=DataInterface() {propertyValues {Curve {keys [<at 0 0>,<at 0.7 0.3>,<at 1 1>]}}}
     CurveIndex=&Divide_Float(
      A=&DistanceBetweenPositions(
       A=&ReturnCameraPosition(Coordinate Space:ENiagaraCoordinateSpace=::Simulation, Camera Query:NiagaraDataInterfaceCamera=DataInterface())
       B=$User.Stop Location
      )
      B=&Multiply_Float(A=1000, B=&MakeFloatFromVector(Vector=$Engine.Owner.Scale, Channel=::NewEnumerator0))
     )
    )
    ScaleA=true
    ScaleRGB=true
   ) {module=ScaleColor001}
  ]
  rendererClasses [&NiagaraSpriteRendererProperties]
  renderers [{rendererClass=&NiagaraSpriteRendererProperties}]
 }
}
summary {
 systemName=NS_MaskMaker_BeamAttack
 emitters [
  {emitterName=EnergyBeam, rendererClasses [&NiagaraRibbonRendererProperties]}
  {emitterName=Light, rendererClasses [&NiagaraSpriteRendererProperties, &NiagaraLightRendererProperties]}
  {emitterName=EnergyChargeMesh, rendererClasses [&NiagaraMeshRendererProperties]}
  {emitterName=SplineRing001, rendererClasses [&NiagaraMeshRendererProperties]}
  {emitterName=ImpactLight, rendererClasses [&NiagaraSpriteRendererProperties, &NiagaraLightRendererProperties]}
  {emitterName=Streaks_Fat, rendererClasses [&NiagaraSpriteRendererProperties]}
  {emitterName=Sparks, simTarget=GPUComputeSim, rendererClasses [&NiagaraSpriteRendererProperties]}
  {emitterName=ScorchDecal, rendererClasses [&NiagaraDecalRendererProperties]}
  {emitterName=EnergyLeak, rendererClasses [&NiagaraMeshRendererProperties]}
  {emitterName=EnergyLeak001, rendererClasses [&NiagaraSpriteRendererProperties]}
 ]
 userVariables [
  User.BeamWidth {type=NiagaraFloat, defaultValue=1, description=""}
  User.DecalNormal {type=Vector3f, description="", defaultValue {struct=Vector3f, value=<1 0 0>}}
  User.OverallOpacity {type=NiagaraFloat, defaultValue=1, description=""}
  User.SpawnDecal {type=NiagaraBool, defaultValue=true}
  User.Stop Location {type=NiagaraPosition, description="", defaultValue {}}
 ]
}
