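# Telegraph effect for the MaskMaker boss's beam attack, marking the attack before it lands.
# A ribbon beam links User Stop Location to the system's position, widening and brightening with a shootout flash value.
# An edge-glow mesh and flickering flare sprites with a light follow the stop point, scaled by time of day, with the flare and light also using beam length or distance LOD.
# A separate decal follows User DecalLocation, fading with a system timeline.
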NiagaraSystem /MaskMakerBoss/VFX/NS_MaskMaker_BeamAttack_Telegraph
dependencies {
 usedDataInterfaces [
  &NiagaraDataInterfaceCurve
  &NiagaraDataInterfaceCamera
  &NiagaraDataInterfaceSkeletalMesh
  &NiagaraDataInterfaceArrayFloat
 ]
 usedDynamicInputs [
  &Multiply_VectorByFloat
  &FN_LOD_Distance_Remap
  &VectorLength
  &Subtract_Position
  &NormalizeVector
  &SubtractVectorFromPosition
  &RemapRange
 ]
 usedModules [
  &self:SystemSpawnScript.SystemScriptSource.SystemScriptGraph.NiagaraNodeAssignment_1.SetVariables_#1
  &SystemState
  &Timeline
  &self:SystemSpawnScript.SystemScriptSource.SystemScriptGraph.NiagaraNodeAssignment_5.SetVariables_#2
  &self:SystemSpawnScript.SystemScriptSource.SystemScriptGraph.NiagaraNodeAssignment_0.SetVariables_#3
  &self:SystemSpawnScript.SystemScriptSource.SystemScriptGraph.NiagaraNodeAssignment_7.SetVariables_#4
  &self:SystemSpawnScript.SystemScriptSource.SystemScriptGraph.NiagaraNodeAssignment_8.SetVariables_#5
  &self:SystemSpawnScript.SystemScriptSource.SystemScriptGraph.NiagaraNodeAssignment_4.SetVariables_#6
 ]
}
emitters {
 EdgeGlow {
  emitterUpdateScript [
   &EmitterState(
    Inactive Response=::NewEnumerator1
    Life Cycle Mode=::NewEnumerator1
    Loop Behavior=::NewEnumerator1
    Loop Duration Mode=::NewEnumerator1
    UseLoopDelay=false
   )
   &SpawnBurst_Instantaneous(Use Spawn Probability=false, UseLoopCountLimit=false) {module=SpawnBurst_Instantaneous001}
  ]
  particleSpawnScript [
   &InitializeParticle(
    Color Mode=::NewEnumerator0
    Lifetime=1.5
    Mesh Scale Mode=::NewEnumerator3
    Position=$User.Stop Location
    Position Mode=::NewEnumerator1
    UsePositionOffset=false
    Write Material Random=true
   )
   &InitialMeshOrientation(Mesh Axis To Orient=<0 0 1>, Mesh Orientation Mode=::NewEnumerator2, Use Rotation Vector=true)
  ]
  particleUpdateScript [
   &ParticleState(Kill Particles When Lifetime Has Elapsed=false, Let Infinitely Lived Particles Die When Emitter Deactivates=true)
   &CameraOffset(Camera Offset Amount=38.6542)
   &ScaleColor(Scale RGB=&VectorFromFloat(Value=0.2), ScaleA=true, ScaleRGB=true)
   /SetVariables(Particles.Position=$User.Stop Location) {isSetParameters=true, module=SetVariables_#7}
   &DynamicMaterialParameters(Module.Index0Param0Write=true, Module.Index0Param1Write=true, Module.Index0Param2Write=true, Module.Index0Param3Write=true)
  ]
  rendererClasses [&NiagaraMeshRendererProperties]
  renderers [{rendererClass=&NiagaraMeshRendererProperties, rendererIndex=0}]
 }
 EnergyBeam {
  emitterUpdateScript [
   &EmitterState()
   &SpawnBurst_Instantaneous(Spawn Count=6, Spawn Time=0.1, Use Spawn Probability=false, UseLoopCountLimit=false)
   &SampleSkeletalMesh(
    AdditionalYawPitchRoll=false
    Mesh Sampling Type:ENiagaraSkelSamplingModeFull=::NewEnumerator5
    Skeletal Mesh:NiagaraDataInterfaceSkeletalMesh=$System.SK_Mesh
    Socket Sampling Mode:ENiagaraSocketSamplingMode=::NewEnumerator5
   )
   /Niagara/Modules/Beams/BeamEmitterSetup(
    Absolute Beam End=true
    Beam End:NiagaraPosition=/Niagara/DynamicInputs/Helpers/SimulationPosition()
    Beam Start:NiagaraPosition=$User.Stop Location
   )
  ]
  particleSpawnScript [
   &InitializeParticle(Mesh Scale Mode=::NewEnumerator3, Mesh Uniform Scale=50, Position Mode=::NewEnumerator0, Write Material Random=true)
   /Niagara/Modules/Beams/SpawnBeam()
   /Niagara/Modules/Beams/BeamWidth(Beam Width:NiagaraFloat=&Multiply_Float(
     A=&Lerp_Float(
      A=&Multiply_Float(
       A=10
       B=&Lerp_Float(A=1, B=&MakeFloatFromVector(Vector=$Engine.Owner.Scale, Channel=::NewEnumerator0), Alpha=1)
      )
      B=&Multiply_Float(
       A=30
       B=&Lerp_Float(A=1, B=&MakeFloatFromVector(Vector=$Engine.Owner.Scale, Channel=::NewEnumerator0), Alpha=0.3)
      )
      Alpha=&FloatFromCurve(
       FloatCurve=DataInterface() {propertyValues {Curve {keys [{tangentMode=RCTM_User, time=0, value=1}, {tangentMode=RCTM_User, time=1, value=0}]}}}
       CurveIndex=$Particles.RibbonLinkOrder
      )
     )
     B=1
    ))
   &InitialMeshOrientation(Mesh Orientation Mode=::NewEnumerator2, Orientation Vector=$System.BeamDirection, Use Rotation Vector=false)
  ]
  particleUpdateScript [
   &ParticleState()
   /Niagara/Modules/Beams/UpdateBeam()
   &DynamicMaterialParameters(
    Index 0 Param 4=&FloatFromCurve(
     FloatCurve=DataInterface() {propertyValues {Curve {keys [{arriveTangent=4.1862, interpMode=RCIM_Cubic, leaveTangent=4.1862, tangentMode=RCTM_User, time=0, value=0}, <at 0.1 1 RCIM_Cubic>]}}}
     CurveIndex=$Emitter.Age
    )
    Index 1 Param 1=&MakeFloatFromVector(Vector=$Emitter.BeamDirection, Channel=::NewEnumerator0)
    Index 1 Param 2=&MakeFloatFromVector(Vector=$Emitter.BeamDirection, Channel=::NewEnumerator1)
    Index 1 Param 3=&MakeFloatFromVector(Vector=$Emitter.BeamDirection, Channel=::NewEnumerator2)
    Module.Index0Param0Write=true
    Module.Index0Param1Write=true
    Module.Index0Param2Write=true
    Module.Index0Param3Write=true
    Module.Index1Param0Write=true
    Module.Index1Param1Write=true
    Module.Index1Param2Write=true
    Module.Index1Param3Write=true
    Write Parameter Index 1=true
   )
   &SolveForcesAndVelocity(Clamp Velocity=false, Limit Acceleration=false)
   &ScaleColor(
    Scale RGB=&VectorFromFloat(Value=&Lerp_Float(
      A=1
      B=10
      Alpha=&RandomRangeFloat(
       Minimum=&FNTimeOfDayScalar(Night Scale=1, Morning Scale=0.6, Day Scale=0.6, Evening Scale=0.6)
       Maximum=1
       Evaluation Type=::NewEnumerator1
      )
     ))
    ScaleA=true
    ScaleRGB=true
   )
   &ScaleColor(
    Scale Alpha=&FloatFromCurve(FloatCurve=DataInterface() {propertyValues {Curve {keys [<at 0 1 RCIM_Cubic>,<at 1 0 RCIM_Cubic>]}}})
    Scale RGB=&VectorFromFloat(Value=&Multiply_Float(A=1, B=&Lerp_Float(A=1, B=20, Alpha=$System.BeamShootoutFlash)))
    ScaleA=true
    ScaleRGB=true
   ) {module=ScaleColor001}
   /Niagara/Modules/Beams/ScaleBeamWidth(Beam Width Scale:NiagaraFloat=&Lerp_Float(A=1, B=1.1, Alpha=$System.BeamShootoutFlash))
  ]
  rendererClasses [&NiagaraRibbonRendererProperties]
  renderers [{rendererClass=&NiagaraRibbonRendererProperties}]
 }
 ImpactLight {
  emitterSpawnScript [/SetVariables(Emitter.CameraOffsetSprite=50) {isSetParameters=true}]
  emitterUpdateScript [
   &EmitterState(
    Inactive Response=::NewEnumerator1
    Life Cycle Mode=::NewEnumerator1
    Loop Behavior=::NewEnumerator1
    Loop Duration Mode=::NewEnumerator1
    UseLoopDelay=false
   )
   &SpawnBurst_Instantaneous(Use Spawn Probability=true, UseLoopCountLimit=true)
   &SpawnBurst_Instantaneous(Use Spawn Probability=false, UseLoopCountLimit=false) {module=SpawnBurst_Instantaneous001}
  ]
  particleSpawnScript [
   &InitializeParticle(
    Lifetime=0.4
    Position=$User.Stop Location
    Position Mode=::NewEnumerator1
    Sprite Rotation Mode=::NewEnumerator2
    Sprite Size Mode=::NewEnumerator3
    Uniform Sprite Size=400
    UsePositionOffset=false
    Write Material Random=true
   )
   &DynamicMaterialParameters(Module.Index0Param0Write=false, Module.Index0Param1Write=false, Module.Index0Param2Write=false, Module.Index0Param3Write=false)
   &CameraOffset(Camera Offset Amount=-100)
  ]
  particleUpdateScript [
   &ParticleState(Kill Particles When Lifetime Has Elapsed=false, Let Infinitely Lived Particles Die When Emitter Deactivates=true)
   &ScaleColor(
    Scale Alpha=&Multiply_Float(
     A=&MakeCustomFloatFromBool(
      Bool=&SetBoolByFloatComparison(A=$System.BeamLength, B=9500, Comparison Type=::NewEnumerator6)
      True Float=0.5
      False Float=0.1
     )
     B=&Multiply_Float(
      A=&FNTimeOfDayScalar(Night Scale=0.2, Morning Scale=0.5, Evening Scale=0.5)
      B=&Lerp_Float(A=0.15, B=1, Alpha=$System.ImpactLODRemap)
     )
    )
    Scale RGB=<1 0.05 0.3>
    ScaleA=true
    ScaleRGB=true
   ) {module=ScaleColor002}
   /SetVariables(Particles.FlareColor=&MultiplyLinearColor(
     Linear Color=<rgba 1 0.0232 0.02 1>
     Scale Factor=&MakeVector4(
      X=1
      Y=1
      Z=1
      W=&Multiply_Float(
       A=&Multiply_Float(
        A=&MakeCustomFloatFromBool(
         Bool=&SetBoolByFloatComparison(A=$System.BeamLength, B=9500, Comparison Type=::NewEnumerator6)
         True Float=1
         False Float=0.05
        )
        B=&Lerp_Float(A=0.3, B=1, Alpha=$System.ImpactLODRemap)
       )
       B=&FloatFromCurve(
        FloatCurve=DataInterface() {propertyValues {Curve {keys [<at 0 0.5>,<at 1 1>]}}}
        CurveIndex=&Divide_Float(
         A=&DistanceBetweenPositions(
          A=&ReturnCameraPosition(Coordinate Space:ENiagaraCoordinateSpace=::Simulation, Camera Query:NiagaraDataInterfaceCamera=DataInterface())
          B=$User.Stop Location
         )
         B=&Multiply_Float(A=300, B=&MakeFloatFromVector(Vector=$Engine.Owner.Scale, Channel=::NewEnumerator0))
        )
        Scale Curve=&FNTimeOfDayScalar(Night Scale=0.3, Morning Scale=0.7, Evening Scale=0.7)
       )
      )
     )
    )) {isSetParameters=true}
   &ApplyOwnerScaleToAttributes(Scale Initial Sprite Size=true, Sprite Size Scale Amount=0.2)
   &SpriteRotationRate(Rotation Rate=800)
   &ScaleSpriteSize(
    Scale Sprite Size Mode=::NewEnumerator0
    Uniform Scale Factor=&Lerp_Float(
     A=0.5
     B=1
     Alpha=&RandomRangeFloat(
      Minimum=&FNTimeOfDayScalar(Night Scale=1, Morning Scale=0.7, Day Scale=0.6, Evening Scale=0.7)
      Maximum=1
      Evaluation Type=::NewEnumerator1
     )
    )
   ) {module=ScaleSpriteSize001}
   &ScaleSpriteSize(
    Scale Sprite Size Mode=::NewEnumerator0
    Uniform Scale Factor=&Multiply_Float(
     A=&Multiply_Float(A=1, B=1)
     B=&Lerp_Float(A=0.5, B=1, Alpha=$System.ImpactLODRemap)
    )
   )
   &FN_Light_Attributes(
    Color=&MultiplyLinearColor(
     Linear Color=<rgba 1 0.0738 0.4269 1>
     Scale Factor=&MakeVector4_FromFloat(Float=&FNTimeOfDayScalar(Night Scale=0.5, Morning Scale=1.5, Day Scale=2, Evening Scale=1.5))
    )
    Create Color Attribute=true
    Create Position Attribute=true
    Create Radius Attribute=true
    Enable Particle Light Enable=true
    Light Enabled=true
    Position=&AddVectorToPosition(
     Position=$Particles.Position
     Vector=&TransformVector(
      Vector=&Multiply_Vector(
       A=Linked() {linkedVariable {name=System.BeamDirection, type=NiagaraPosition}}
       B=&VectorFromFloat(Value=0)
      )
      Source Space=::World
      Destination Space=::Local
     )
    )
    Radius=&Multiply_Float(
     A=&Lerp_Float(A=400, B=800, Alpha=&RandomRangeFloat(Minimum=0.5, Maximum=1, Evaluation Type=::NewEnumerator1))
     B=&Lerp_Float(A=0.2, B=0.5, Alpha=$System.ImpactLODRemap)
    )
    Scale Color=20
   )
   /SetVariables(Particles.Position=$User.Stop Location) {isSetParameters=true, module=SetVariables_#7}
  ]
  rendererClasses [&NiagaraSpriteRendererProperties, &NiagaraLightRendererProperties]
  renderers [
   {rendererClass=&NiagaraSpriteRendererProperties}
   {rendererClass=&NiagaraLightRendererProperties, rendererIndex=1}
   {rendererClass=&NiagaraSpriteRendererProperties, rendererIndex=2}
  ]
 }
 TelegraphDecal {
  emitterUpdateScript [
   &EmitterState(
    Inactive Response=::NewEnumerator1
    Life Cycle Mode=::NewEnumerator1
    Loop Behavior=::NewEnumerator1
    Loop Duration Mode=::NewEnumerator1
    UseLoopDelay=false
   )
   /SetVariables(Emitter.DecalInitialOpacity:NiagaraFloat=&Lerp_Float(
     A=0.4
     B=0
     Alpha=&FloatFromCurve(FloatCurve=DataInterface() {propertyValues {Curve {keys [<at 0 0 RCIM_Cubic>,<at 1 1 RCIM_Cubic>]}}}, CurveIndex=$System.BeamLengthScalar)
    )) {isSetParameters=true}
   &SpawnBurst_Instantaneous(Spawn Count=1, Use Spawn Probability=false)
  ]
  particleSpawnScript [
   &InitializeParticle(Lifetime=3, Sprite Size Mode=::NewEnumerator3, Uniform Sprite Size=600, UsePositionOffset=false, Write Material Random=true)
   /SetVariables(
    Particles.DecalFade=1
    Particles.DecalOrientation:Quat4f=&OrientMeshToVector(Look At Direction=<0 0 1>, Reference Axis Vector=<1 0 0>)
    Particles.DecalSize:Vector3f=<150 400 400>
   ) {isSetParameters=true}
  ]
  particleUpdateScript [
   &ParticleState(Kill Particles When Lifetime Has Elapsed=false, Let Infinitely Lived Particles Die When Emitter Deactivates=true)
   /SetVariables(Particles.DecalFade:NiagaraFloat=&Lerp_Float(
     A=0
     B=$System.Timeline.Float
     Alpha=&FloatFromCurve(FloatCurve=DataInterface() {propertyValues {Curve {keys [
          {arriveTangent=12.3499, interpMode=RCIM_Cubic, leaveTangent=12.3499, tangentMode=RCTM_User, time=0, value=0}
          {arriveTangent=0.2129, interpMode=RCIM_Cubic, leaveTangent=0.2129, tangentMode=RCTM_User, time=0.2, value=1}
          {arriveTangent=-0.0037, interpMode=RCIM_Cubic, leaveTangent=-0.0037, tangentMode=RCTM_User, time=1, value=1}
         ]}}})
    )) {isSetParameters=true}
   /SetVariables(Particles.Position=$User.DecalLocation) {isSetParameters=true, module=SetVariables_#7}
  ]
  rendererClasses [&NiagaraDecalRendererProperties]
  renderers [{rendererClass=&NiagaraDecalRendererProperties, rendererIndex=0}]
 }
}
summary {
 systemName=NS_MaskMaker_BeamAttack_Telegraph
 emitters [
  {emitterName=EnergyBeam, rendererClasses [&NiagaraRibbonRendererProperties]}
  {emitterName=ImpactLight, rendererClasses [&NiagaraSpriteRendererProperties, &NiagaraLightRendererProperties]}
  {emitterName=EdgeGlow, rendererClasses [&NiagaraMeshRendererProperties]}
  {emitterName=TelegraphDecal, rendererClasses [&NiagaraDecalRendererProperties]}
 ]
 userVariables [
  User.DecalLocation {type=NiagaraPosition, description="", defaultValue {}}
  User.IsAboutToAttack {type=NiagaraBool}
  User.Stop Location {type=NiagaraPosition, description="", defaultValue {}}
 ]
}
