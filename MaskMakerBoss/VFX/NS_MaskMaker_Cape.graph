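# Niagara effect for the MaskMaker boss's cape, spawned as a burst of seven mesh particles that live for User.CapeLifetime.
# Partition-indexed arrays set each piece's rotation, scale and mirroring.
# A spring-driven pendulum with drag swings the particles, and interpolation produces a lagged position.
# That lagged direction, the rest axis, lifetime curves and a side flag feed dynamic material parameters, while alpha follows its own fade curve.
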NiagaraSystem /MaskMakerBoss/VFX/NS_MaskMaker_Cape
dependencies {
 usedDataInterfaces [
  &NiagaraDataInterfaceArrayFloat3
  &NiagaraDataInterfaceCurve
  &NiagaraDataInterfaceArrayFloat
  &NiagaraDataInterfaceSkeletalMesh
 ]
 usedDynamicInputs [
  &RemapRange
  &VectorLength
  &MakeCustomVectorFromBool
  &SetBoolByFloatComparison
 ]
 usedModules [
  &self:SystemSpawnScript.SystemScriptSource.SystemScriptGraph.NiagaraNodeAssignment_2.SetVariables_#1
  &SystemState
  &self:SystemSpawnScript.SystemScriptSource.SystemScriptGraph.NiagaraNodeAssignment_0.SetVariables_#2
  &self:SystemSpawnScript.SystemScriptSource.SystemScriptGraph.NiagaraNodeAssignment_3.SetVariables_#3
  &self:SystemSpawnScript.SystemScriptSource.SystemScriptGraph.NiagaraNodeAssignment_4.SetVariables_#4
 ]
}
emitters {SimpleSpriteBurst001 {
  emitterUpdateScript [
   &EmitterState(Life Cycle Mode=::NewEnumerator1, Loop Behavior=::NewEnumerator1, UseLoopDelay=false)
   &SpawnBurst_Instantaneous(Spawn Count=7, Use Spawn Probability=false)
  ]
  particleSpawnScript [
   &PartitionParticles(Number Of Partitions=8)
   &InitializeParticle(
    Color Mode=::NewEnumerator0
    Lifetime=$User.CapeLifetime
    Mesh Scale=&SelectVectorFromArray(
     Direct Array Index=$Particles.PartitionParticles.Partition
     Vector Selection Array=DataInterface() {propertyValues {FloatData [<2 2 2>, <2 2 2>, <2 2 2>, <2 2 2>, <2 -2 -2>, <2 -2 -2>, <2 -2 -2>, <2 -2 -2>]}}
    )
    Mesh Scale Mode=::NewEnumerator4
    UsePositionOffset=false
    Write Material Random=true
   ) {module=InitializeParticle001}
   &InitialMeshOrientation(
    Mesh Orientation Mode=::NewEnumerator2
    Orientation Vector=<0 1 0>
    Rotation=&SelectVectorFromArray(
     Direct Array Index=$Particles.PartitionParticles.Partition
     Vector Selection Array=DataInterface() {propertyValues {FloatData [
        <0.4403 -0.1468 0.3644>
        <-0.0488 0.5161 0.6288>
        <-0.1203 0.4804 0.468>
        <-0.2186 0.7128 0.468>
        <0.5612 -0.5353 0.1036>
        <-0.287 0.3921 0.164>
        <1.9523 -1.6995 -0.2286>
        <-1.159 1.2842 -1.7725>
       ]}}
    )
    Use Rotation Vector=true
   )
   &InitialMeshOrientation(Mesh Orientation Mode=::NewEnumerator2, Use Rotation Vector=true) {module=InitialMeshOrientation001}
   /SetVariables(Particles.Length=300) {isSetParameters=true}
  ]
  particleUpdateScript [
   &ParticleState()
   /SetVariables(Particles.MeshOrigin:NiagaraPosition=/Niagara/DynamicInputs/Helpers/SimulationPosition()) {isSetParameters=true}
   &ScaleMeshSize(Scale Factor=&SelectVectorFromArray(
     Direct Array Index=$Particles.PartitionParticles.Partition
     Vector Selection Array=DataInterface() {propertyValues {FloatData [<1.2 1 1>, <1.5 1 1>, <1 1 1>, <1 1 1>, <1.5 1 1>, <1 1 1>, <1.2 1 1>, <1 1 1>]}}
    ))
   &ScaleMeshSize(Scale Factor=&VectorFromFloat(Value=1.5)) {module=ScaleMeshSize001}
   &InterpolateOverTime(Rate Of Change=12, Target Vector Value:NiagaraPosition=$Particles.Position, Variable Type=::NewEnumerator2)
   &ScaleColor(
    Scale Alpha=&Lerp_Float(
     A=0
     B=1
     Alpha=&FloatFromCurve(FloatCurve=DataInterface() {propertyValues {Curve {keys [<at 0 1>, {interpMode=RCIM_Cubic, leaveTangent=-9.1412, tangentMode=RCTM_Break, time=0.7, value=1}, <at 1 0>]}}})
    )
    ScaleA=true
    ScaleRGB=true
   )
   &PendulumSetup(
    Module.Enable Spring Driver=true
    Pendulum Length:NiagaraFloat=$Particles.Length
    Pendulum Rest Axis:Vector3f=&TransformVector(Vector=<0 -1 0>, Destination Space=::Simulation)
   )
   &Drag(Drag=8, Rotational Drag=5, Use Linear Drag=true, Use Rotational Drag=true) {module=Drag001}
   &SolveForcesAndVelocity(Clamp Velocity=false, Limit Acceleration=false)
   &PendulumConstraint()
   &CalculateAccurateVelocity()
   &self:ScratchModule(Position:NiagaraPosition=$StackContext.InterpolateOverTime.InterpolateOverTime_V3_.Moving Average, Vector:Vector3f=<0 -1 0>) {enabled=false, module=ScratchModule}
   /SetVariables(Particles.InterpolateOverTime.InterpolateOverTime_V3_.Moving Average:Vector3f=$Particles.ScratchModule.Position) {enabled=false, isSetParameters=true}
   /SetVariables(Particles.MeshFacing:Vector3f=&NormalizeVector(
     Vector To Normalize=&Subtract_Vector(
      A=Linked() {linkedVariable {name=Particles.MeshOrigin, type=NiagaraPosition}}
      B=Linked() {linkedVariable {name=Particles.Position, type=NiagaraPosition}}
     )
     Safe Normalize=true
     Safe Fallback Threshold=0
     Safe Fallback Vector=<0 0 0.001>
    )) {isSetParameters=true}
   /SetVariables(Particles.Vector:Vector3f=&SubtractVectorFromPosition(
     Position=$StackContext.InterpolateOverTime.InterpolateOverTime_V3_.Moving Average
     Vector=&AddVectorToPosition(
      Position=$Engine.Emitter.SimulationPosition
      Vector=&Multiply_VectorByFloat(Vector=&TransformVector(Vector=<0 -1 0>, Destination Space=::Simulation), Float=$Particles.Length)
     )
    )) {isSetParameters=true}
   /SetVariables(Particles.InterpolateDy:Vector4f=&MakeFloat4FromVectorAndFloat(
     "Vector (RGB)"=$Particles.Vector
     "Float (Alpha)"=&Lerp_Float(A=2, B=5, Alpha=$System.InterpolateOverTime.InterpolateOverTime_Float.Moving Average)
    )) {isSetParameters=true}
   /SetVariables(Particles.Vector:Vector3f=&SubtractVectorFromPosition(
     Position=$Particles.Position
     Vector=&AddVectorToPosition(
      Position=$Engine.Emitter.SimulationPosition
      Vector=&Multiply_VectorByFloat(Vector=&TransformVector(Vector=<0 -1 0>, Destination Space=::Simulation), Float=200)
     )
    )) {enabled=false, isSetParameters=true, module=SetVariables_#5}
   &DynamicMaterialParameters(
    Index 0 Param 1=&Lerp_Float(
     A=0
     B=1
     Alpha=&FloatFromCurve(FloatCurve=DataInterface() {propertyValues {Curve {keys [
          {arriveTangent=6.166, interpMode=RCIM_Cubic, leaveTangent=6.166, tangentMode=RCTM_User, time=0, value=0}
          {arriveTangent=1.5564, interpMode=RCIM_Cubic, leaveTangent=1.5564, tangentMode=RCTM_User, time=0.1236, value=0.7495}
          {arriveTangent=0.0711, leaveTangent=0.0711, time=0.2595, value=1}
          <at 0.6252 1>
          {arriveTangent=-6.8167, interpMode=RCIM_Cubic, leaveTangent=-6.8167, time=0.7524, value=0.8199}
          {arriveTangent=-8.1467, interpMode=RCIM_Cubic, leaveTangent=-8.1467, time=0.8487, value=-0.5235}
          <at 0.9758 -1 RCIM_Cubic>
         ]}}})
    )
    Index 0 Param 2=&Lerp_Float(
     A=1
     B=5
     Alpha=&FloatFromCurve(FloatCurve=DataInterface() {propertyValues {Curve {keys [<at 0.0149 0 RCIM_Cubic>,<at 0.0917 0.95 RCIM_Cubic>,<at 0.3126 0 RCIM_Cubic>]}}})
    )
    Index 0 Param 3=&Multiply_Float(
     A=&Lerp_Float(
      A=1
      B=1.5
      Alpha=&FloatFromCurve(FloatCurve=DataInterface() {propertyValues {Curve {keys [<at 0.0149 0 RCIM_Cubic>,<at 0.0917 0.95 RCIM_Cubic>,<at 0.3126 0 RCIM_Cubic>]}}})
     )
     B=1.5
    )
    Index 0 Param 4=&SelectFloatFromArray(
     Direct Array Index=$Particles.PartitionParticles.Partition
     Float Selection Array=DataInterface() {propertyValues {FloatData [1,1,1,1,0,0,0,0]}}
    )
    Index 2 Vector 1:Vector3f=&TransformVector(Vector=<0 -1 0>, Destination Space=::Simulation)
    Index 3 Vector 1:Vector3f=&NormalizeVector(
     Vector To Normalize=&SubtractVectorFromPosition(
      Position=$StackContext.InterpolateOverTime.InterpolateOverTime_V3_.Moving Average
      Vector=Linked() {linkedVariable {name=Engine.Emitter.SimulationPosition, type=NiagaraPosition}}
     )
     Safe Normalize=true
     Safe Fallback Threshold=0
     Safe Fallback Vector=<0 0 0.001>
    )
    Module.Index0Float0Write=true
    Module.Index0Float1Write=true
    Module.Index0Float2Write=true
    Module.Index0Float3Write=true
    Module.Index2Float3Write=true
    Module.Index2VectorWrite=true
    Module.Index3Float3Write=true
    Module.Index3VectorWrite=true
    Write Parameter Index 2=::NewEnumerator3
    Write Parameter Index 3=::NewEnumerator3
   )
   /Niagara/Modules/Update/Orientation/V2/OrientMeshToVector(
    Look At Direction:Vector3f=&TransformVector(Vector=<0 -1 0>, Destination Space=::Simulation)
    Override Quat=false
    Side Axis Direction:Vector3f=<0 0 0>
    Up Axis Direction:Vector3f=&TransformVector(Vector=<0 0 1>, Destination Space=::Simulation)
    UseOptionalSideAxis=false
    UseOptionalUpAxis=true
   )
  ]
  rendererClasses [&NiagaraSpriteRendererProperties, &NiagaraMeshRendererProperties]
  renderers [
   {rendererClass=&NiagaraSpriteRendererProperties}
   {rendererClass=&NiagaraMeshRendererProperties, rendererIndex=1}
  ]
 }}
summary {
 systemName=NS_MaskMaker_Cape
 emitters [{
   emitterName=SimpleSpriteBurst001
   rendererClasses [&NiagaraSpriteRendererProperties, &NiagaraMeshRendererProperties]
  }]
 userVariables [User.CapeLifetime {type=NiagaraFloat, defaultValue=2, description=""}]
}
