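# Niagara system for the MaskMaker boss teleport VFX: dark smoke, chromatic refraction and a short sphere burst at User Offset Pos.
# SmokeLoopIdle drives meshes with age-based emitter curves and pulls sphere-spawned smoke particles toward the offset.
# ChromaticRefraction animates a sprite material through Timeline and picks red or violet via WanderingBoss; SphereBurst scales alpha by time of day.
# Global Scale sizes smoke and refraction; SmokeLoop_LW has no renderers.
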NiagaraSystem /MaskMakerBoss/VFX/NS_MaskMaker_Teleport
dependencies {
 usedDataInterfaces [
  &NiagaraDataInterfaceCurve
  &NiagaraDataInterfaceColorCurve
  &NiagaraDataInterfaceVectorCurve
  &NiagaraDataInterfaceArrayFloat
 ]
 usedDynamicInputs []
 usedModules [&SystemState]
}
emitters {
 ChromaticRefraction {
  emitterUpdateScript [
   &EmitterState(Life Cycle Mode=::NewEnumerator1, Loop Behavior=::NewEnumerator1, Loop Duration=2, UseLoopDelay=false)
   &SpawnBurst_Instantaneous(Use Spawn Probability=false, UseLoopCountLimit=false)
   /SetVariables(Emitter.A_1:LinearColor=&MakeCustomLinearColorFromBool(Bool=$User.WanderingBoss, True Color=<rgba 1 0.078 0.0934 1>, False Color=<rgba 0.2879 0 1 1>)) {isSetParameters=true}
   /SetVariables(Emitter.A_2:LinearColor=&MakeCustomLinearColorFromBool(Bool=$User.WanderingBoss, True Color=<rgba 1 0.092 0.2585 1>, False Color=<rgba 0.3412 0 1 1>)) {isSetParameters=true}
   /SetVariables() {isSetParameters=true}
  ]
  particleSpawnScript [
   &InitializeParticle(
    Mesh Scale Mode=::NewEnumerator3
    Sprite Size Mode=::NewEnumerator3
    Uniform Sprite Size=&Multiply_Float(A=10, B=$User.Global Scale)
    UsePositionOffset=false
    Write Material Random=true
   )
   &SystemLocation(MaskSpawnGroup=false, Offset:NiagaraPosition=$User.Offset Pos)
  ]
  particleUpdateScript [
   &ParticleState()
   &Timeline(
    Play Rate=&FloatFromCurve(
     FloatCurve=DataInterface() {propertyValues {Curve {keys [
         {interpMode=RCIM_Cubic, leaveTangent=-12.5334, tangentMode=RCTM_Break, time=0, value=4}
         {interpMode=RCIM_Cubic, tangentMode=RCTM_Break, time=0.7, value=1}
        ]}}}
     Scale Curve=0.5
    )
    Timeline End=1.0000000200408773e+20
   )
   &ScaleSpriteSize(
    Scale Sprite Size Mode=::NewEnumerator0
    Uniform Scale Factor=&FloatFromCurve(
     FloatCurve=DataInterface() {propertyValues {Curve {keys [
         {interpMode=RCIM_Cubic, tangentMode=RCTM_User, time=0, value=0}
         {interpMode=RCIM_Cubic, tangentMode=RCTM_User, time=0.5, value=1}
         {arriveTangent=-2, interpMode=RCIM_Cubic, leaveTangent=-2, time=0.9, value=0.8}
         {arriveTangent=-7.0911, tangentMode=RCTM_User, time=1, value=0}
        ]}}}
     Scale Curve=70
    )
   )
   &DynamicMaterialParameters(
    Index 0 Param 1=2
    Index 0 Param 2=5
    Index 0 Param 3=2
    Index 0 Param 4=1
    Index 1 Param 2=500
    Index 1 Param 3=0.5
    Index 1 Param 4=1
    Index 2 Param 1=$Particles.Timeline.Time
    Index 2 Param 2=$Particles.Timeline.Time
    Index 2 Param 3=$Particles.Timeline.Time
    Index 2 Param 4=&SelectFloatFromArray(Direct Array Index=$Particles.PartitionParticles.Partition, Float Selection Array=DataInterface() {propertyValues {FloatData [1,0.3,0.3]}})
    Index 3 Param 1=&FloatFromCurve(FloatCurve=DataInterface() {propertyValues {Curve {keys [<at 0.1 1>,<at 0.4 0>]}}})
    Index 3 Param 2=200
    Module.Index0Param0Write=true
    Module.Index0Param1Write=true
    Module.Index0Param2Write=true
    Module.Index0Param3Write=true
    Module.Index1Param0Write=true
    Module.Index1Param1Write=true
    Module.Index1Param2Write=true
    Module.Index1Param3Write=true
    Module.Index2Param0Write=true
    Module.Index2Param1Write=true
    Module.Index2Param2Write=true
    Module.Index2Param3Write=true
    Module.Index3Param0Write=true
    Module.Index3Param1Write=true
    Module.Index3Param2Write=true
    Module.Index3Param3Write=true
    Write Parameter Index 1=true
    Write Parameter Index 2=true
    Write Parameter Index 3=true
   ) {module=DynamicMaterialParameters001}
   &ScaleColor(Scale RGB=&VectorFromFloat(Value=0.5), ScaleA=true, ScaleRGB=true) {module=ScaleColor001}
  ]
  rendererClasses [&NiagaraSpriteRendererProperties]
  renderers [{rendererClass=&NiagaraSpriteRendererProperties}]
 }
 SmokeLoopIdle {
  emitterSpawnScript [/SetVariables(Emitter.Position:NiagaraPosition=&TransformVector(Vector=Linked() {linkedVariable {name=User.Offset Pos, type=NiagaraPosition}}, Destination Space=::World, Transform Vector as Position=true)) {isSetParameters=true}]
  emitterUpdateScript [
   &EmitterState(
    Life Cycle Mode=::NewEnumerator1
    Loop Behavior=::NewEnumerator1
    Loop Delay=0.5
    Loop Duration=$User.LoopDurationOfMainEmitter
    UseLoopDelay=false
   )
   &SpawnBurst_Instantaneous(Use Spawn Probability=false, UseLoopCountLimit=false)
   &SpawnRate(SpawnRate=20, Use Spawn Probability=false)
   &SpawnRate(SpawnGroup=1, SpawnRate=20, Use Spawn Probability=false) {module=SpawnRate001}
   /SetVariables(Emitter.SharedCurve1=DataInterface() {propertyValues {Curve {keys [
        {arriveTangent=1.1977, interpMode=RCIM_Cubic, leaveTangent=1.1977, tangentMode=RCTM_User, time=0, value=0}
        {arriveTangent=1.914, interpMode=RCIM_Cubic, leaveTangent=0.736, tangentMode=RCTM_Break, time=0.1432, value=0.1703}
        {arriveTangent=-0.7631, interpMode=RCIM_Cubic, leaveTangent=0.0241, tangentMode=RCTM_Break, time=0.2949, value=0.8528}
        {arriveTangent=-1.0789, interpMode=RCIM_Cubic, leaveTangent=-1.0789, time=0.8, value=0.6}
        {arriveTangent=-3, interpMode=RCIM_Cubic, leaveTangent=-3, time=0.9, value=0.2}
        {arriveTangent=-0.8856, interpMode=RCIM_Cubic, leaveTangent=-0.8856, tangentMode=RCTM_User, time=1, value=0}
       ]}}}) {isSetParameters=true}
   /SetVariables(
    Emitter.CameraOffset=-25
    Emitter.CameraOfs_Outline=-25
    Emitter.Color:LinearColor=&ColorFromCurve(
     ColorCurve=DataInterface() {propertyValues {
       AlphaCurve {keys [<at 0 0 RCIM_Cubic>,<at 0.0826 2 RCIM_Cubic>]}
       BlueCurve {keys [<at 0.4494 0.0045 RCIM_Cubic>]}
       GreenCurve {keys [<at 0.4494 0.0006 RCIM_Cubic>]}
       RedCurve {keys [<at 0.4494 0.01 RCIM_Cubic>]}
      }}
     CurveIndex=$Emitter.Age
    )
    Emitter.Color2:LinearColor=&ColorFromCurve(
     ColorCurve=DataInterface() {propertyValues {
       AlphaCurve {keys [<at 0 0 RCIM_Cubic>,<at 0.0826 4 RCIM_Cubic>]}
       BlueCurve {keys [<at 0.4494 0.0067 RCIM_Cubic>]}
       GreenCurve {keys [<at 0.4494 0.0009 RCIM_Cubic>]}
       RedCurve {keys [<at 0.4494 0.015 RCIM_Cubic>]}
      }}
     CurveIndex=$Emitter.Age
    )
    Emitter.Color_Outline:LinearColor=<rgba 1 0 0.1667 5>
   ) {isSetParameters=true}
   /SetVariables(
    Emitter.CoreCameraOfs:NiagaraFloat=&FloatFromCurve(
     FloatCurve=DataInterface() {propertyValues {Curve {keys [{arriveTangent=-2.2178, interpMode=RCIM_Cubic, leaveTangent=-2.2178, tangentMode=RCTM_User, time=0.1, value=1}, <at 0.6 -1>]}}}
     CurveIndex=$Emitter.NormalizedLoopAge
     Scale Curve=-50
    )
    Emitter.DynamicParam0_inky:Vector4f=&MakeVector4(
     X=&MakeFloatFromVector4(
      Vector 4=&MakeVector4(
       X=&FloatFromCurve(
        FloatCurve=DataInterface() {propertyValues {Curve {keys [
            {arriveTangent=4.1069, interpMode=RCIM_Cubic, leaveTangent=4.1069, tangentMode=RCTM_User, time=0, value=0}
            {interpMode=RCIM_Cubic, tangentMode=RCTM_User, time=0.4, value=1.1}
            {interpMode=RCIM_Cubic, leaveTangent=-1.8333, time=1, value=1}
           ]}}}
        CurveIndex=$Emitter.Age
        Scale Curve=0.5
       )
       Y=&RandomRangeFloat(Minimum=0, Maximum=1)
       Z=&FloatFromCurve(
        FloatCurve=DataInterface() {propertyValues {Curve {keys [{tangentMode=RCTM_User, time=0, value=0}, {tangentMode=RCTM_User, time=1, value=1}]}}}
        CurveIndex=$Emitter.Age
       )
       W=&FloatFromCurve(
        FloatCurve=DataInterface() {propertyValues {Curve {keys [<at 0.45 0 RCIM_Cubic>, {arriveTangent=11.9729, tangentMode=RCTM_User, time=0.99, value=10}]}}}
        CurveIndex=$Emitter.Age
       )
      )
      Channel=::NewEnumerator0
     )
     Y=&RandomRangeFloat(Minimum=0, Maximum=1)
     Z=&FloatFromCurve(
      FloatCurve=DataInterface() {propertyValues {Curve {keys [
          {arriveTangent=1.7411, interpMode=RCIM_Cubic, leaveTangent=1.7411, tangentMode=RCTM_User, time=0.1, value=0}
          {tangentMode=RCTM_User, time=0.8, value=1}
         ]}}}
      CurveIndex=$Emitter.Age
     )
     W=&FloatFromCurve(
      FloatCurve=DataInterface() {propertyValues {Curve {keys [<at 0.3 0 RCIM_Cubic>, {arriveTangent=29.1781, tangentMode=RCTM_User, time=0.99, value=10}]}}}
      CurveIndex=$Emitter.Age
     )
    )
    Emitter.DynamicParam3:Vector4f=&MakeVector4(
     X=&FloatFromCurve(
      FloatCurve=DataInterface() {propertyValues {Curve {keys [
          {arriveTangent=1.1977, interpMode=RCIM_Cubic, leaveTangent=1.1977, tangentMode=RCTM_User, time=0, value=0}
          {arriveTangent=1.914, interpMode=RCIM_Cubic, leaveTangent=0.736, tangentMode=RCTM_Break, time=0.1432, value=0.1703}
          {arriveTangent=-0.7631, interpMode=RCIM_Cubic, leaveTangent=0.0241, tangentMode=RCTM_Break, time=0.2949, value=0.8528}
          {arriveTangent=-1.1761, interpMode=RCIM_Cubic, leaveTangent=-1.1761, time=0.73, value=0.7}
          {arriveTangent=-2.2243, interpMode=RCIM_Cubic, leaveTangent=-2.2243, tangentMode=RCTM_User, time=0.85, value=0.2}
          {arriveTangent=-0.8856, interpMode=RCIM_Cubic, leaveTangent=-0.8856, tangentMode=RCTM_User, time=1, value=0}
         ]}}}
      CurveIndex=$Emitter.Age
      Scale Curve=1.5
     )
     Y=&FloatFromCurve(
      FloatCurve=DataInterface() {propertyValues {Curve {keys [
          {arriveTangent=0.3228, interpMode=RCIM_Cubic, leaveTangent=0.3228, tangentMode=RCTM_User, time=0, value=0}
          {arriveTangent=1.771, interpMode=RCIM_Cubic, leaveTangent=1.771, tangentMode=RCTM_User, time=0.1953, value=0.2}
          {arriveTangent=1.7297, interpMode=RCIM_Cubic, leaveTangent=1.7297, time=0.343, value=1.05}
          {arriveTangent=-2.6609, interpMode=RCIM_Cubic, leaveTangent=-2.6609, time=0.6, value=0.9}
          {arriveTangent=-1.0556, interpMode=RCIM_Cubic, leaveTangent=-1.0556, tangentMode=RCTM_User, time=0.7, value=0.1}
          {tangentMode=RCTM_User, time=1, value=0}
         ]}}}
      CurveIndex=$Emitter.Age
      Scale Curve=25
     )
     Z=&RandomRangeFloat(Minimum=0, Maximum=1)
     W=1
    )
    Emitter.Scale:Vector3f=&Multiply_VectorByFloat(
     Vector=&Multiply_Vector(
      A=&VectorFromFloat(Value=&FloatFromCurve(FloatCurve=$Emitter.SharedCurve1, CurveIndex=$Emitter.Age, Scale Curve=$User.Global Scale))
      B=<1.12 1.12 0.4>
     )
     Float=1
    )
    Emitter.Scale2:Vector3f=&Multiply_VectorByFloat(
     Vector=&Multiply_Vector(
      A=&VectorFromFloat(Value=&FloatFromCurve(FloatCurve=$Emitter.SharedCurve1, CurveIndex=$Emitter.Age, Scale Curve=1.2))
      B=<0.72 0.72 0.32>
     )
     Float=$User.Global Scale
    )
    Emitter.Scale3:Vector3f=&Multiply_VectorByFloat(
     Vector=&Multiply_Vector(
      A=&VectorFromFloat(Value=&FloatFromCurve(FloatCurve=$Emitter.SharedCurve1, CurveIndex=$Emitter.Age, Scale Curve=1.1))
      B=<0.7 0.7 0.3>
     )
     Float=$User.Global Scale
    )
    Emitter.SpriteSize:Vector2f=&Multiply_Vector2D_ByFloat(
     Float=&MakeFloatFromVector(
      Vector=&Multiply_Vector(
       A=&VectorFromFloat(Value=&FloatFromCurve(
         FloatCurve=DataInterface() {propertyValues {Curve {keys [
             {arriveTangent=0.3782, interpMode=RCIM_Cubic, leaveTangent=0.3782, tangentMode=RCTM_User, time=0, value=0.5}
             {arriveTangent=0.3146, interpMode=RCIM_Cubic, leaveTangent=-0.8436, tangentMode=RCTM_Break, time=1.1, value=1}
            ]}}}
         CurveIndex=$Emitter.Age
         Scale Curve=0.8
        ))
       B=<1.4 1.4 0.4>
      )
      Channel=::NewEnumerator0
     )
     Vector2D=&Multiply_Vector2D_ByFloat(Float=$User.Global Scale, Vector2D=<320 320>)
    )
   ) {isSetParameters=true}
   /SetVariables() {isSetParameters=true}
  ]
  particleSpawnScript [
   &InitializeParticle(
    Color=&MaskLinearColorBySpawnGroup(Masked Linear Color=<rgba 0 0 0 1>, Passthrough Linear Color=<rgba 0.2604 0.0521 0.0764 1>)
    Lifetime=&RandomRangeFloat(Minimum=0.8, Maximum=1.3)
    Sprite Rotation Mode=::NewEnumerator2
    Sprite Size=&Multiply_Vector2D_ByFloat(
     Float=$User.Global Scale
     Vector2D=&Vector2DFromFloat(Value=&RandomRangeFloat(Minimum=135, Maximum=160))
    )
    Sprite Size Mode=::NewEnumerator4
    Sprite UV Mode=::NewEnumerator3
    UsePositionOffset=false
    Write Material Random=true
   )
   &CameraOffset(Camera Offset Amount=-100)
   &DynamicMaterialParameters(
    Index 0 Param 1=&RandomRangeFloat(Minimum=0, Maximum=-6)
    Module.Index0Param0Write=true
    Module.Index0Param1Write=false
    Module.Index0Param2Write=false
    Module.Index0Param3Write=false
   )
   &SystemLocation(MaskSpawnGroup=false, Offset:NiagaraPosition=$User.Offset Pos)
   &ShapeLocation(
    MaskSpawnGroup=false
    Non Uniform Scale=<0.7 0.7 1>
    Override Seed=false
    Sphere Radius=&Multiply_Float(A=110, B=$User.Global Scale)
    UseOwnerScale=true
   )
   &AddVelocity(Velocity=&RandomRangeVector(Minimum=<0 0 30>, Maximum=<0 0 -30>))
  ]
  particleUpdateScript [
   &ParticleState()
   &ScaleSpriteSize(
    Scale Sprite Size Mode=::NewEnumerator0
    Uniform Scale Factor=&FloatFromCurve(FloatCurve=DataInterface() {propertyValues {Curve {keys [
         <at 0 0.1 RCIM_Cubic>
         {arriveTangent=0.9, interpMode=RCIM_Cubic, leaveTangent=0.9, time=0.4, value=0.4}
         {interpMode=RCIM_Cubic, leaveTangent=0.1472, time=1, value=1}
        ]}}})
   )
   &ScaleColor(
    Scale Alpha=&FloatFromCurve(
     FloatCurve=DataInterface() {propertyValues {Curve {keys [
         <at 0 0 RCIM_Cubic>
         {arriveTangent=2.2188, interpMode=RCIM_Cubic, leaveTangent=2.2188, time=0.2, value=1}
         {arriveTangent=-1.25, interpMode=RCIM_Cubic, leaveTangent=-1.25, time=0.8, value=1.775}
         {interpMode=RCIM_Cubic, leaveTangent=-26.9591, time=1, value=0}
        ]}}}
     Scale Curve=1.5
    )
    Scale RGB=&VectorFromCurve(VectorCurve=DataInterface() {propertyValues {
       XCurve {keys [<at 0.25 1 RCIM_Cubic>,<at 1 0.5>]}
       YCurve {keys [<at 0.25 1 RCIM_Cubic>,<at 1 0.5>]}
       ZCurve {keys [<at 0.25 1 RCIM_Cubic>,<at 1 0.5>]}
      }})
    ScaleA=true
    ScaleRGB=true
   )
   &ScaleMeshSize(Scale Factor=&VectorFromFloat(Value=&FloatFromCurve(FloatCurve=DataInterface() {propertyValues {Curve {keys [
          {arriveTangent=6.6826, interpMode=RCIM_Cubic, leaveTangent=6.6826, tangentMode=RCTM_User, time=0, value=0}
          {arriveTangent=1.914, interpMode=RCIM_Cubic, leaveTangent=0.736, tangentMode=RCTM_Break, time=0.0951, value=0.9887}
          {arriveTangent=-0.7631, interpMode=RCIM_Cubic, leaveTangent=-1.2047, tangentMode=RCTM_Break, time=0.864, value=0.9964}
          {arriveTangent=-5.5008, interpMode=RCIM_Cubic, leaveTangent=-5.5008, tangentMode=RCTM_User, time=1, value=0}
         ]}}})))
   /Niagara/Modules/Update/SubUV/SubUVAnimation(Number Of Frames=32)
   &DynamicMaterialParameters(
    Index 0 Param 1=&FloatFromCurve(
     FloatCurve=DataInterface() {propertyValues {Curve {keys [
         {arriveTangent=-0.0338, interpMode=RCIM_Cubic, leaveTangent=-0.0338, tangentMode=RCTM_User, time=0, value=0}
         {arriveTangent=1.9274, interpMode=RCIM_Cubic, leaveTangent=1.9274, tangentMode=RCTM_User, time=1, value=1}
        ]}}}
     Scale Curve=0.1
    )
    Index 0 Param 2=50
    Module.Index0Param0Write=true
    Module.Index0Param1Write=true
    Module.Index0Param2Write=false
    Module.Index0Param3Write=false
   ) {module=DynamicMaterialParameters001}
   &PointAttractionForce(
    Attraction Radius=500
    Attraction Strength=10
    Attractor Position:NiagaraPosition=$User.Offset Pos
    Kill Within Radius=false
    Module.Use Falloff=true
    UseAttractorOffset=false
   )
   &SolveForcesAndVelocity(Clamp Velocity=false, Limit Acceleration=false)
  ]
  rendererClasses [&NiagaraMeshRendererProperties, &NiagaraSpriteRendererProperties]
  renderers [
   {rendererClass=&NiagaraMeshRendererProperties, rendererIndex=0}
   {rendererClass=&NiagaraMeshRendererProperties, rendererIndex=1}
   {rendererClass=&NiagaraSpriteRendererProperties, rendererIndex=2}
  ]
 }
 SmokeLoop_LW {rendererClasses [], renderers []}
 SphereBurst {
  emitterUpdateScript [
   &EmitterState()
   &SpawnBurst_Instantaneous(Spawn Group=1, Use Spawn Probability=false, UseLoopCountLimit=false) {module=SpawnBurst_Instantaneous001}
   /SetVariables(
    Emitter.Color:LinearColor=&ColorFromCurve(
     ColorCurve=DataInterface() {propertyValues {
       AlphaCurve {keys [<at 0 0>,<at 0.0924 1 RCIM_Cubic>,<at 0.3264 0 RCIM_Cubic>]}
       BlueCurve {keys [<at 0.0339 0>,<at 0.3334 0 RCIM_Cubic>]}
       GreenCurve {keys [<at 0.0339 3.99>,<at 0.3334 0 RCIM_Cubic>]}
       RedCurve {keys [<at 0.0339 40>,<at 0.3334 0 RCIM_Cubic>]}
      }}
     CurveIndex=$Emitter.Age
    )
    Emitter.Position:NiagaraPosition=<0 0 100>
    Emitter.Radius=450
   ) {isSetParameters=true}
  ]
  particleSpawnScript [
   /SetVariables(Particles.VisibilityTag=$Emitter.SpawnGroup) {isSetParameters=true}
   &InitializeParticle(
    Color=<rgba 5 0 0.4167 1>
    Lifetime=&SelectFloatFromArray(Direct Array Index=$Emitter.SpawnGroup, Float Selection Array=DataInterface() {propertyValues {FloatData [0.5,0.2,0.3]}})
    Mesh Scale Mode=::NewEnumerator3
    Mesh Uniform Scale=0.5
    UsePositionOffset=false
    Write Material Random=true
   )
   &SystemLocation(MaskSpawnGroup=false, Offset:NiagaraPosition=$User.Offset Pos)
   &InitialMeshOrientation(
    Rotation=&RandomRangeVector(Minimum=<0 0 0>, Maximum=<1 1 1>)
    Use Orientation Vector=true
    Use Rotation Vector=true
   )
   /SetVariables(Particles.Color2=<rgba 50 37.4589 6.8714 2>) {isSetParameters=true}
  ]
  particleUpdateScript [
   &ParticleState()
   &ScaleSpriteSize(Uniform Curve Sprite Scale=DataInterface() {propertyValues {Curve {keys [
        {arriveTangent=4.1785, interpMode=RCIM_Cubic, leaveTangent=4.1785, tangentMode=RCTM_User, time=0, value=0.1}
        {arriveTangent=1.1179, interpMode=RCIM_Cubic, leaveTangent=1.1179, tangentMode=RCTM_User, time=0.2042, value=0.6228}
        {arriveTangent=0.0224, interpMode=RCIM_Cubic, leaveTangent=0.0224, tangentMode=RCTM_User, time=1, value=1}
       ]}}})
   &ScaleColor(
    Scale Alpha=&FloatFromCurve(
     FloatCurve=DataInterface() {propertyValues {Curve {keys [{arriveTangent=45, interpMode=RCIM_Cubic, time=0, value=0}, <at 0.3 1.75 RCIM_Cubic>, <at 1 0 RCIM_Cubic>]}}}
     Scale Curve=&FNTimeOfDayScalar(Night Scale=0.1, Morning Scale=0.2, Day Scale=0.25, Evening Scale=0.15)
    )
    Scale RGB=<3 3 3>
    ScaleA=true
    ScaleRGB=true
   )
   &ScaleMeshSize(Scale Factor=&VectorFromFloat(Value=&FloatFromCurve(FloatCurve=DataInterface() {propertyValues {Curve {keys [
          {arriveTangent=6.5452, interpMode=RCIM_Cubic, leaveTangent=6.5452, tangentMode=RCTM_User, time=0, value=0.05}
          {arriveTangent=6.229, interpMode=RCIM_Cubic, leaveTangent=6.229, tangentMode=RCTM_User, time=0.1, value=0.75}
          {arriveTangent=2.2956, interpMode=RCIM_Cubic, leaveTangent=2.2956, tangentMode=RCTM_User, time=0.1698, value=0.1791}
          {arriveTangent=-0.0405, interpMode=RCIM_Cubic, leaveTangent=-0.0405, tangentMode=RCTM_User, time=0.6, value=0.7688}
         ]}}})))
   &DynamicMaterialParameters(
    Index 0 Param 1=&FloatFromCurve(FloatCurve=DataInterface() {propertyValues {Curve {keys [
         {arriveTangent=-0.3119, interpMode=RCIM_Cubic, leaveTangent=-0.3119, tangentMode=RCTM_User, time=0, value=0.825}
         {interpMode=RCIM_Cubic, tangentMode=RCTM_User, time=0.2, value=0.675}
         {arriveTangent=0.5138, interpMode=RCIM_Cubic, leaveTangent=0.5138, tangentMode=RCTM_User, time=0.5, value=0.85}
         {interpMode=RCIM_Cubic, tangentMode=RCTM_User, time=1, value=1}
        ]}}})
    Index 0 Param 2=&FloatFromCurve(FloatCurve=DataInterface() {propertyValues {Curve {keys [<at 0 15 RCIM_Cubic>, {arriveTangent=-40.0295, interpMode=RCIM_Cubic, leaveTangent=-40.0295, tangentMode=RCTM_User, time=1, value=0}]}}})
    Module.Param0WriteEnabled=true
    Module.Param1WriteEnabled=true
    Module.Param2WriteEnabled=true
    Module.Param3WriteEnabled=true
   )
   /SetVariables(
    Particles.Color1:LinearColor=&ColorFromCurve(ColorCurve=DataInterface() {propertyValues {
       AlphaCurve {keys [<at 0.2659 2 RCIM_Cubic>,<at 0.4948 0.5>,<at 1 0 RCIM_Cubic>]}
       BlueCurve {keys [<at 0.0735 0.25 RCIM_Cubic>,<at 0.4828 0.003 RCIM_Cubic>]}
       GreenCurve {keys [<at 0.0735 0.1 RCIM_Cubic>,<at 0.4828 0 RCIM_Cubic>]}
       RedCurve {keys [<at 0.0735 1 RCIM_Cubic>,<at 0.4828 0.0365 RCIM_Cubic>]}
      }})
    Particles.DMP0A:Vector4f=&MakeVector4(
     X=1
     Y=&FloatFromCurve(FloatCurve=DataInterface() {propertyValues {Curve {keys [{interpMode=RCIM_Cubic, tangentMode=RCTM_User, time=0, value=0.25}, {interpMode=RCIM_Cubic, tangentMode=RCTM_User, time=1, value=-1}]}}})
     Z=0.5
     W=50
    )
    Particles.DMP2A:Vector4f=<1 1 1 1>
    Particles.MeshSize1:Vector3f=&Multiply_VectorByFloat(
     Vector=<9 9 9>
     Float=&FloatFromCurve(FloatCurve=DataInterface() {propertyValues {Curve {keys [
          {arriveTangent=5.7182, interpMode=RCIM_Cubic, leaveTangent=5.7182, tangentMode=RCTM_User, time=0, value=0.1}
          {arriveTangent=1.9532, interpMode=RCIM_Cubic, leaveTangent=1.9532, tangentMode=RCTM_User, time=0.2, value=0.9}
          {arriveTangent=-0.0405, interpMode=RCIM_Cubic, leaveTangent=-0.0405, tangentMode=RCTM_User, time=1, value=1.25}
         ]}}})
    )
   ) {isSetParameters=true}
  ]
  rendererClasses [&NiagaraMeshRendererProperties]
  renderers [{rendererClass=&NiagaraMeshRendererProperties, rendererIndex=0}]
 }
}
summary {
 systemName=NS_MaskMaker_Teleport
 emitters [
  {emitterName=SmokeLoop_LW, rendererClasses []}
  {emitterName=SmokeLoopIdle, rendererClasses [&NiagaraMeshRendererProperties, &NiagaraSpriteRendererProperties]}
  {emitterName=ChromaticRefraction, rendererClasses [&NiagaraSpriteRendererProperties]}
  {emitterName=SphereBurst, rendererClasses [&NiagaraMeshRendererProperties]}
 ]
 userVariables [
  User.Global Scale {type=NiagaraFloat, defaultValue=1, description=""}
  User.LoopDurationOfMainEmitter {type=NiagaraFloat, defaultValue=1, description=""}
  User.Offset Pos {type=NiagaraPosition, description="", defaultValue {}}
  User.Scalar {type=NiagaraFloat, defaultValue=2, description=""}
  User.WanderingBoss {type=NiagaraBool}
 ]
}
