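# Niagara effect for the MaskMaker boss's teleport AOE charge-up, drawing a dark ribbon trail between user-supplied source and target positions.
# A single emitter delays a two-particle burst, placing linked ribbon endpoints at both positions, raised slightly in Z.
# Two ribbon renderers display the trail, with a randomized U pan offset set at spawn.
# Lifetime curves drive a dynamic material parameter, a brief bright flash that quickly dims, and alpha fades.
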NiagaraSystem /MaskMakerBoss/VFX/NS_MaskMaker_TeleportAOE_Chargeup
dependencies {usedDataInterfaces [&NiagaraDataInterfaceCurve], usedDynamicInputs [], usedModules [&SystemState]}
emitters {DarkTeleportTrail {
  emitterSpawnScript [/SetVariables(Emitter.PanSpeedU=-2) {isSetParameters=true}]
  emitterUpdateScript [
   &EmitterState()
   &SpawnBurst_Instantaneous(Spawn Count=2, Spawn Time=1.5, Use Spawn Probability=false, UseLoopCountLimit=false)
  ]
  particleSpawnScript [
   /SetVariables(Particles.RibbonLinkOrder:NiagaraFloat=&ReturnNormalizedExecIndex()) {isSetParameters=true}
   &InitializeParticle(
    Lifetime=3.5
    Position=&LerpPosition(
     A=&AddVectorToPosition(Position=$User.SourcePosition, Vector=&MakeVector(X=0, Y=0, Z=50))
     B=&AddVectorToPosition(Position=$User.TargetPosition, Vector=&MakeVector(X=0, Y=0, Z=50))
     Alpha=&ReturnNormalizedExecIndex()
    )
    Position Mode=::NewEnumerator1
    Ribbon Width=400
    Ribbon Width Mode=::NewEnumerator1
    Sprite Size Mode=::NewEnumerator3
    Uniform Sprite Size=5
    UsePositionOffset=false
    Write Material Random=true
   )
   &FN_RibbonMaterial(Color Pan U Offset:NiagaraFloat=&RandomRangeFloat(Minimum=0, Maximum=1))
  ]
  particleUpdateScript [
   &ParticleState()
   &DynamicMaterialParameters(
    Index 0 Param 1=&FloatFromCurve(FloatCurve=DataInterface() {propertyValues {Curve {keys [
         {arriveTangent=1.9867, interpMode=RCIM_Cubic, leaveTangent=1.9867, tangentMode=RCTM_User, time=0.0032, value=0.1483}
         {arriveTangent=0.8544, interpMode=RCIM_Cubic, leaveTangent=0.8544, time=0.3224, value=0.7112}
         <at 1 1 RCIM_Cubic>
        ]}}})
    Index 0 Param 3=1
    Module.Index0Float0Write=true
    Module.Index0Float1Write=true
    Module.Index0Float2Write=true
    Module.Index0Float3Write=false
   )
   &ScaleColor(
    Scale Alpha=&FloatFromCurve(FloatCurve=DataInterface() {propertyValues {Curve {keys [
         {arriveTangent=-1.0879, interpMode=RCIM_Cubic, time=0, value=0}
         {arriveTangent=0.7, interpMode=RCIM_Cubic, leaveTangent=0.7, time=0.02, value=0.85}
         {arriveTangent=-0.8673, interpMode=RCIM_Cubic, leaveTangent=-0.8673, time=0.5, value=0.35}
         {interpMode=RCIM_Cubic, leaveTangent=-8.0089, time=1, value=0}
        ]}}})
    Scale RGB=&VectorFromFloat(Value=&Lerp_Float(
      A=0.25
      B=5
      Alpha=&FloatFromCurve(FloatCurve=DataInterface() {propertyValues {Curve {
          defaultValue=1
          keys [
           {arriveTangent=-6.7126, interpMode=RCIM_Cubic, leaveTangent=-6.7126, tangentMode=RCTM_User, time=0, value=1}
           {interpMode=RCIM_Cubic, tangentMode=RCTM_User, time=0.1, value=0}
          ]
         }}})
     ))
    ScaleA=true
    ScaleRGB=true
   ) {module=ScaleColor002}
   /SetVariables(Particles.DarkTrailColor:LinearColor=&MakeLinearColorFromVectorAndFloat(
     "Vector (RGB)"=<1 1 1>
     "Float (Alpha)"=&FloatFromCurve(FloatCurve=DataInterface() {propertyValues {Curve {
         defaultValue=1
         keys [{interpMode=RCIM_Cubic, tangentMode=RCTM_User, time=0, value=1}, {interpMode=RCIM_Cubic, tangentMode=RCTM_User, time=1, value=0}]
        }}})
    )) {isSetParameters=true}
  ]
  rendererClasses [&NiagaraRibbonRendererProperties]
  renderers [
   {rendererClass=&NiagaraRibbonRendererProperties}
   {rendererClass=&NiagaraRibbonRendererProperties, rendererIndex=1}
  ]
 }}
summary {
 systemName=NS_MaskMaker_TeleportAOE_Chargeup
 emitters [{emitterName=DarkTeleportTrail, rendererClasses [&NiagaraRibbonRendererProperties]}]
 userVariables [
  User.IsLocalPlayer {type=NiagaraBool}
  User.RandomArray {
   type=NiagaraDataInterfaceArrayFloat
   defaultValue {
    struct=NiagaraExt_VariableValue_DataInterface
    value {
     dataInterface=&self:NiagaraDataInterfaceArrayFloat_2
     dataInterfaceClass=&NiagaraDataInterfaceArrayFloat
    }
   }
  }
  User.SourcePosition {type=NiagaraPosition, description="", defaultValue {}}
  User.TargetPosition {type=NiagaraPosition, description="", defaultValue {}}
 ]
}
