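# Niagara effect for the start of the Mask Maker boss's teleport.
# A mesh portal, flare-and-smoke burst and twisting absorb sprites track the skeletal mesh's pelvis until about 0.43 seconds, when source movement scale drops to zero.
# A delayed, noise-driven ribbon trail spawns by distance moved, with spacing widening over time.
# Curves animate size, color and dissolve-style material parameters, with dark secondary layers and wandering-boss hue variants.
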NiagaraSystem /MaskMakerBoss/VFX/NS_MaskMaker_Teleport_Start
dependencies {
 usedDataInterfaces [
  &NiagaraDataInterfaceSkeletalMesh
  &NiagaraDataInterfaceCurve
  &NiagaraDataInterfaceColorCurve
  &NiagaraDataInterfaceVector2DCurve
 ]
 usedDynamicInputs [&FN_LOD_Distance_Remap]
 usedModules [
  &self:SystemSpawnScript.SystemScriptSource.SystemScriptGraph.NiagaraNodeAssignment_0.SetVariables_#1
  &self:SystemSpawnScript.SystemScriptSource.SystemScriptGraph.NiagaraNodeAssignment_1.SetVariables_#2
  &SystemState
  &self:SystemSpawnScript.SystemScriptSource.SystemScriptGraph.NiagaraNodeAssignment_2.SetVariables_#3
 ]
}
emitters {
 FlareSmoke {
  emitterUpdateScript [
   &EmitterState(Life Cycle Mode=::NewEnumerator1, Loop Behavior=::NewEnumerator1, Loop Duration=2, UseLoopDelay=true)
   &SpawnBurst_Instantaneous(Use Spawn Probability=false, UseLoopCountLimit=false)
  ]
  particleSpawnScript [
   &InitializeParticle(
    Lifetime=0.75
    Mass Mode=::NewEnumerator3
    Mesh Scale=&Multiply_VectorByFloat(Vector=<1 1 1>, Float=0.25)
    Mesh Scale Mode=::NewEnumerator4
    Sprite Size=<900 500>
    Sprite Size Mode=::NewEnumerator4
    UsePositionOffset=false
    Write Material Random=true
   )
   &SystemLocation(
    MaskSpawnGroup=false
    Offset=&AddVectorToPosition(Position=$System.ImpactOffset, Vector=<0 0 -3>)
    Offset Coordinate Space=::World
   )
   &CameraOffset(Camera Offset Amount=60)
   &InitialMeshOrientation(Orientation Axis=<0 0 1>, Orientation Vector=<0 0 1>, Rotation=<0 0.25 0>, Use Orientation Vector=true, Use Rotation Vector=true)
   &InitialMeshOrientation(
    Orientation Axis=<0 0 0>
    Orientation Vector=<0 0 0>
    Rotation=&RandomRangeVector(Minimum=<0 0 0>, Maximum=<1 0 0>)
    Use Orientation Vector=false
    Use Rotation Vector=true
   ) {module=InitialMeshOrientation001}
   /SetVariables(Particles.CameraOffsetFlare=120) {isSetParameters=true}
   /SetVariables(Particles.SizeDark:Vector3f=<2.5 2.5 2.5>) {isSetParameters=true}
  ]
  particleUpdateScript [
   &ParticleState()
   &SkeletalMeshLocation(
    Bone Sampling Mode=::NewEnumerator4
    MaskSpawnGroup=false
    Orientation Sampling=::NewEnumerator1
    Position Sampling=::NewEnumerator1
    Sampled Position Offset=<50 0 0>
    Skeletal Mesh=DataInterface() {propertyValues {
      DefaultMesh=null
      ExcludeBoneName=root
      OptionalComponentClass=null
      PreviewMesh=/Game/Characters/Player/Male/Medium/Base/SK_M_MALE_Base_Skeleton
      SkinningMode=SkinOnTheFly
      SoftSourceActor=null
      SourceMode=Default
      UvSetIndex=0
      WholeMeshLOD=-1
      bExcludeBone=false
      bReadDeformedGeometry=true
      bRequireCurrentFrameData=true
      ComponentTags []
      FilteredBones [pelvis]
      FilteredSockets []
      MeshUserParameter {parameter {name=None, typeDefHandle {registeredTypeIndex=1}, varData []}}
      SamplingRegions []
     }}
    Source Movement Scale:Vector3f=&VectorFromFloat(Value=&FloatFromCurve(
      FloatCurve=DataInterface() {propertyValues {Curve {keys [{interpMode=RCIM_Cubic, tangentMode=RCTM_User, time=0.42, value=1}, {interpMode=RCIM_Cubic, tangentMode=RCTM_User, time=0.44, value=0}]}}}
      CurveIndex=$Emitter.Age
     ))
   )
   &ScaleSpriteSize(
    Scale Factor=&MakeVector2D(
     X=&FloatFromCurve(
      FloatCurve=DataInterface() {propertyValues {Curve {keys [{arriveTangent=27.5059, interpMode=RCIM_Cubic, leaveTangent=27.5059, tangentMode=RCTM_User, time=0.1, value=0.1}, <at 0.2 1.1>]}}}
      Scale Curve=1.5
     )
     Y=&FloatFromCurve(
      FloatCurve=DataInterface() {propertyValues {Curve {keys [
          {arriveTangent=-55.1607, interpMode=RCIM_Cubic, leaveTangent=-55.1607, tangentMode=RCTM_User, time=0.02, value=3}
          {arriveTangent=-12.8632, interpMode=RCIM_Cubic, leaveTangent=-12.8632, tangentMode=RCTM_User, time=0.07, value=1.2}
          {interpMode=RCIM_Cubic, tangentMode=RCTM_User, time=0.3, value=0.2}
         ]}}}
      Scale Curve=1.5
     )
    )
    Scale Sprite Size Mode=::NewEnumerator2
   )
   &ScaleColor(
    Scale Alpha=&FloatFromCurve(
     FloatCurve=DataInterface() {propertyValues {Curve {keys [
         {arriveTangent=0.4784, interpMode=RCIM_Cubic, leaveTangent=0.4784, tangentMode=RCTM_User, time=0, value=0}
         {arriveTangent=-3.8452, interpMode=RCIM_Cubic, leaveTangent=-3.8452, tangentMode=RCTM_User, time=0.1, value=1}
         {arriveTangent=-0.1713, interpMode=RCIM_Cubic, leaveTangent=-0.1713, tangentMode=RCTM_User, time=1, value=0}
        ]}}}
     Scale Curve=1.3
    )
    Scale RGB=&MakeVectorFromLinearColorRGB(Color=&HueShiftLinearColor(
      Hue=&MakeCustomFloatFromBool(Bool=$System.IsWanderingBoss, True Float=-0.24, False Float=0)
      Saturation=&MakeCustomFloatFromBool(Bool=$System.IsWanderingBoss, True Float=0.7, False Float=1)
      Value=1
      Linear Color=&ColorFromCurve(ColorCurve=DataInterface() {propertyValues {
         AlphaCurve {keys [<at 0 1 RCIM_Cubic>]}
         BlueCurve {keys [<at 0.0169 0.307 RCIM_Cubic>,<at 0.6322 0.04>]}
         GreenCurve {keys [<at 0.0169 0.01 RCIM_Cubic>,<at 0.6322 0.006>]}
         RedCurve {keys [<at 0.0169 1 RCIM_Cubic>,<at 0.6322 0.0088>]}
        }})
     ))
    ScaleA=true
    ScaleRGB=true
   )
   /SetVariables(Particles.ColorDark:LinearColor=&MakeLinearColorFromVectorAndFloat(
     "Vector (RGB)"=&MakeVectorFromLinearColorRGB(Color=&HueShiftLinearColor(
       Hue=&MakeCustomFloatFromBool(Bool=$System.IsWanderingBoss, True Float=-0.24, False Float=0)
       Saturation=&MakeCustomFloatFromBool(Bool=$System.IsWanderingBoss, True Float=0.7, False Float=1)
       Value=1
       Linear Color=&ColorFromCurve(ColorCurve=DataInterface() {propertyValues {
          AlphaCurve {keys [<at 0 1 RCIM_Cubic>]}
          BlueCurve {keys [<at 0 0.09 RCIM_Cubic>,<at 0.4371 0.016>]}
          GreenCurve {keys [<at 0 0 RCIM_Cubic>,<at 0.4371 0.0021>]}
          RedCurve {keys [<at 0 0.3 RCIM_Cubic>,<at 0.4371 0.0356>]}
         }})
      ))
     "Float (Alpha)"=&FloatFromCurve(
      FloatCurve=DataInterface() {propertyValues {Curve {keys [
          {interpMode=RCIM_Cubic, tangentMode=RCTM_User, time=0, value=0}
          {arriveTangent=-2.4988, interpMode=RCIM_Cubic, leaveTangent=-2.4988, tangentMode=RCTM_User, time=0.3, value=1}
          {interpMode=RCIM_Cubic, tangentMode=RCTM_User, time=1, value=0}
         ]}}}
      Scale Curve=3
     )
    )) {isSetParameters=true}
   &SolveForcesAndVelocity(Clamp Velocity=false, Limit Acceleration=false)
   &DynamicMaterialParameters(
    Index 0 Param 1=-0.1
    Index 0 Param 2=&Lerp_Float(
     A=0
     B=1
     Alpha=&FloatFromCurve(FloatCurve=DataInterface() {propertyValues {Curve {keys [<at 0 0>,<at 1 1>]}}})
    )
    Index 0 Param 3=0.5
    Index 0 Param 4=&Lerp_Float(
     A=0.45
     B=0.8
     Alpha=&FloatFromCurve(FloatCurve=DataInterface() {propertyValues {Curve {keys [
          {arriveTangent=0.2526, interpMode=RCIM_Cubic, leaveTangent=0.2526, tangentMode=RCTM_User, time=0, value=0}
          {arriveTangent=1.7785, interpMode=RCIM_Cubic, leaveTangent=1.7785, tangentMode=RCTM_User, time=1, value=1}
         ]}}})
    )
    Module.Index0Param0Write=true
    Module.Index0Param1Write=true
    Module.Index0Param2Write=true
    Module.Index0Param3Write=true
   )
   /SetVariables(Particles.DynamicMaterialParameterDark:Vector4f=&MakeVector4(
     X=-0.5
     Y=&Lerp_Float(
      A=0.2
      B=1.5
      Alpha=&FloatFromCurve(FloatCurve=DataInterface() {propertyValues {Curve {keys [<at 0 0>,<at 1 1>]}}})
     )
     Z=1
     W=&Lerp_Float(
      A=-0.5
      B=1
      Alpha=&FloatFromCurve(FloatCurve=DataInterface() {propertyValues {Curve {keys [
           {arriveTangent=0.2526, interpMode=RCIM_Cubic, leaveTangent=0.2526, tangentMode=RCTM_User, time=0.2, value=0}
           {arriveTangent=1.7785, interpMode=RCIM_Cubic, leaveTangent=1.7785, tangentMode=RCTM_User, time=1, value=1}
          ]}}})
     )
    )) {isSetParameters=true}
   /SetVariables(
    Particles.ColorFlare:LinearColor=&ColorFromCurve(ColorCurve=DataInterface() {propertyValues {
       AlphaCurve {keys [<at 0 0>,<at 0.0482 1 RCIM_Cubic>,<at 0.3413 0 RCIM_Cubic>]}
       BlueCurve {keys [<at 0.0373 0.4167 RCIM_Cubic>,<at 0.1975 0.1667 RCIM_Cubic>]}
       GreenCurve {keys [<at 0.0373 0 RCIM_Cubic>,<at 0.1975 0 RCIM_Cubic>]}
       RedCurve {keys [<at 0.0373 5 RCIM_Cubic>,<at 0.1975 1 RCIM_Cubic>]}
      }})
    Particles.PositionFlare:NiagaraPosition=&AddVectorToPosition(
     Position=$Particles.Position
     Vector=&MakeVector(
      X=0
      Y=0
      Z=&Lerp_Float(
       A=0
       B=20
       Alpha=&FloatFromCurve(FloatCurve=DataInterface() {propertyValues {Curve {keys [<at 0 0>,<at 0.65 1>]}}})
      )
     )
    )
   ) {isSetParameters=true}
   /SetVariables(
    Particles.ColorDark:LinearColor=&ColorFromCurve(ColorCurve=DataInterface() {propertyValues {
       AlphaCurve {keys [<at 0 0>,<at 0.0592 1.5 RCIM_Cubic>,<at 0.6507 0 RCIM_Cubic>]}
       BlueCurve {keys [<at 0 0.016 RCIM_Cubic>]}
       GreenCurve {keys [<at 0 0.0021 RCIM_Cubic>]}
       RedCurve {keys [<at 0 0.0356 RCIM_Cubic>]}
      }})
    Particles.SpriteSizeDark:Vector2f=&Vector2DFromFloat(Value=&Lerp_Float(
      A=0
      B=200
      Alpha=&FloatFromCurve(
       FloatCurve=DataInterface() {propertyValues {Curve {keys [
           {interpMode=RCIM_Cubic, tangentMode=RCTM_User, time=0, value=0}
           {interpMode=RCIM_Cubic, tangentMode=RCTM_User, time=0.1, value=1}
           {interpMode=RCIM_Cubic, tangentMode=RCTM_User, time=0.7, value=0}
          ]}}}
       Scale Curve=1.25
      )
     ))
   ) {isSetParameters=true}
   /SetVariables(Particles.SizeDark:Vector3f=&Multiply_VectorByFloat(
     Vector=$Particles.Initial.SizeDark
     Float=&FloatFromCurve(
      FloatCurve=DataInterface() {propertyValues {Curve {keys [
          {interpMode=RCIM_Cubic, tangentMode=RCTM_Break, time=0, value=0.2}
          {arriveTangent=2.9365, interpMode=RCIM_Cubic, leaveTangent=2.9365, tangentMode=RCTM_User, time=0.1, value=1}
          {interpMode=RCIM_Cubic, tangentMode=RCTM_Break, time=1, value=2}
         ]}}}
      Scale Curve=0.85
     )
    )) {isSetParameters=true}
   /SetVariables(Particles.ColorMeshDark:LinearColor=&MultiplyLinearColor(
     Linear Color=<rgba 0.03 0.0021 0.0103 5>
     Scale Factor=&MakeFloat4FromVectorAndFloat(
      "Vector (RGB)"=&VectorFromFloat(Value=&FloatFromCurve(FloatCurve=DataInterface() {propertyValues {Curve {keys [
            {interpMode=RCIM_Cubic, tangentMode=RCTM_Break, time=0, value=10}
            {arriveTangent=-6.6072, interpMode=RCIM_Cubic, leaveTangent=-6.6072, tangentMode=RCTM_User, time=0.15, value=3}
            {interpMode=RCIM_Cubic, tangentMode=RCTM_Break, time=1, value=1}
           ]}}}))
      "Float (Alpha)"=&FloatFromCurve(
       FloatCurve=DataInterface() {propertyValues {Curve {keys [
           {arriveTangent=-3.1255, interpMode=RCIM_Cubic, leaveTangent=-3.1255, tangentMode=RCTM_User, time=0.6, value=1}
           {interpMode=RCIM_Cubic, tangentMode=RCTM_User, time=1, value=0}
          ]}}}
       Scale Curve=&Lerp_Float(A=0.1, B=0.2, Alpha=$System.DistanceScalarNear)
      )
     )
    )) {isSetParameters=true}
   /SetVariables(Particles.DMP_01:Vector4f=&MakeVector4(
     X=&FloatFromCurve(FloatCurve=DataInterface() {propertyValues {Curve {keys [{arriveTangent=-2.1015, interpMode=RCIM_Cubic, leaveTangent=-2.1015, tangentMode=RCTM_User, time=0, value=0.5}, <at 1 -0.15>]}}})
     Y=&FloatFromCurve(FloatCurve=DataInterface() {propertyValues {Curve {keys [{arriveTangent=-0.3464, interpMode=RCIM_Cubic, leaveTangent=-0.3464, tangentMode=RCTM_User, time=0, value=1}, <at 1 0.9>]}}})
     Z=&FloatFromCurve(FloatCurve=DataInterface() {propertyValues {Curve {keys [
          {interpMode=RCIM_Cubic, tangentMode=RCTM_Break, time=0, value=0}
          {arriveTangent=0.4478, interpMode=RCIM_Cubic, leaveTangent=0.4478, tangentMode=RCTM_User, time=0.15, value=1}
          {interpMode=RCIM_Cubic, tangentMode=RCTM_Break, time=1, value=1.2}
         ]}}})
     W=1
    )) {isSetParameters=true}
  ]
  rendererClasses [&NiagaraSpriteRendererProperties, &NiagaraMeshRendererProperties]
  renderers [
   {rendererClass=&NiagaraSpriteRendererProperties}
   {rendererClass=&NiagaraSpriteRendererProperties, rendererIndex=1}
   {rendererClass=&NiagaraMeshRendererProperties, rendererIndex=2}
  ]
 }
 PortalBig {
  emitterUpdateScript [
   &EmitterState(
    Life Cycle Mode=::NewEnumerator1
    Loop Behavior=::NewEnumerator1
    Loop Delay=&Add_Float(A=0, B=$System.FXDelay)
    Loop Duration=1
    UseLoopDelay=true
   )
   &SpawnBurst_Instantaneous(Use Spawn Probability=false, UseLoopCountLimit=false)
   /SetVariables(Emitter.PortalHueShift:NiagaraFloat=&MakeCustomFloatFromBool(Bool=$User.WanderingBoss, True Float=0.2, False Float=0)) {isSetParameters=true}
  ]
  particleSpawnScript [&InitializeParticle(
    Lifetime=0.55
    Mass Mode=::NewEnumerator3
    Mesh Scale Mode=::NewEnumerator3
    Mesh Uniform Scale=1.5
    Position Offset=<0 0 100>
    Sprite Rotation Mode=::NewEnumerator2
    Sprite Size Mode=::NewEnumerator1
    Uniform Sprite Size Max=12
    Uniform Sprite Size Min=6
    UsePositionOffset=false
    Write Material Random=true
   )]
  particleUpdateScript [
   &SkeletalMeshLocation(
    Bone Sampling Mode=::NewEnumerator4
    MaskSpawnGroup=false
    Orientation Sampling=::NewEnumerator1
    Position Sampling=::NewEnumerator1
    Sampled Position Offset=<50 0 0>
    Skeletal Mesh=DataInterface() {propertyValues {
      DefaultMesh=null
      ExcludeBoneName=root
      OptionalComponentClass=null
      PreviewMesh=&SK_M_MALE_Base_Skeleton
      SkinningMode=SkinOnTheFly
      SoftSourceActor=null
      SourceMode=Default
      UvSetIndex=0
      WholeMeshLOD=-1
      bExcludeBone=false
      bReadDeformedGeometry=true
      bRequireCurrentFrameData=true
      ComponentTags []
      FilteredBones [pelvis]
      FilteredSockets []
      MeshUserParameter {parameter {name=None, typeDefHandle {registeredTypeIndex=1}, varData []}}
      SamplingRegions []
     }}
    Source Movement Scale:Vector3f=&VectorFromFloat(Value=&FloatFromCurve(
      FloatCurve=DataInterface() {propertyValues {Curve {keys [{interpMode=RCIM_Cubic, tangentMode=RCTM_User, time=0.42, value=1}, {interpMode=RCIM_Cubic, tangentMode=RCTM_User, time=0.44, value=0}]}}}
      CurveIndex=$Emitter.Age
     ))
   )
   &ParticleState()
   /SetVariables(Particles.ScaleCurve:NiagaraFloat=&FloatFromCurve(
     FloatCurve=DataInterface() {propertyValues {Curve {keys [
         {arriveTangent=0.7733, interpMode=RCIM_Cubic, leaveTangent=0.7733, tangentMode=RCTM_User, time=0, value=0}
         {arriveTangent=-0.7631, interpMode=RCIM_Cubic, leaveTangent=-0.0298, tangentMode=RCTM_Break, time=0.1, value=1}
         {arriveTangent=-0.1589, interpMode=RCIM_Cubic, leaveTangent=-0.1589, tangentMode=RCTM_User, time=0.22, value=1}
         {arriveTangent=-3.9263, interpMode=RCIM_Cubic, leaveTangent=-3.9263, tangentMode=RCTM_User, time=0.38, value=1.5}
         {arriveTangent=-0.5636, interpMode=RCIM_Cubic, leaveTangent=-0.5636, tangentMode=RCTM_User, time=0.9, value=0.8}
         {arriveTangent=-8.1801, interpMode=RCIM_Cubic, leaveTangent=-8.1801, tangentMode=RCTM_User, time=1, value=0.3}
        ]}}}
     Scale Curve=&Multiply_Float(A=0.8, B=$User.GlobalScale)
    )) {isSetParameters=true}
   /SetVariables(
    Particles.Color=&ColorFromCurve(ColorCurve=DataInterface() {propertyValues {
       AlphaCurve {keys [<at 0 0 RCIM_Cubic>,<at 0.0826 2 RCIM_Cubic>]}
       BlueCurve {keys [<at 0.4494 0.0067 RCIM_Cubic>]}
       GreenCurve {keys [<at 0.4494 0.0009 RCIM_Cubic>]}
       RedCurve {keys [<at 0.4494 0.015 RCIM_Cubic>]}
      }})
    Particles.Color2=&ColorFromCurve(ColorCurve=DataInterface() {propertyValues {
       AlphaCurve {keys [<at 0 0 RCIM_Cubic>,<at 0.0826 4 RCIM_Cubic>]}
       BlueCurve {keys [<at 0.4494 0.0067 RCIM_Cubic>]}
       GreenCurve {keys [<at 0.4494 0.0009 RCIM_Cubic>]}
       RedCurve {keys [<at 0.4494 0.015 RCIM_Cubic>]}
      }})
    Particles.Color_Outline:LinearColor=<rgba 6 0 1 1>
    Particles.Scale=&Multiply_VectorByFloat(Vector=<1.12 1.12 0.4>, Float=$Particles.ScaleCurve)
    Particles.Scale2:Vector3f=&Multiply_VectorByFloat(Vector=<0.52 0.52 0.36>, Float=$Particles.ScaleCurve)
    Particles.ScaleOutline:Vector3f=&Multiply_VectorByFloat(Vector=<0.72 0.72 0.32>, Float=$Particles.ScaleCurve)
   ) {isSetParameters=true}
   /SetVariables(Particles.CameraOffset=160, Particles.CameraOffsetOutline=150) {isSetParameters=true}
   &SolveForcesAndVelocity(Clamp Velocity=false, Limit Acceleration=false)
   &DynamicMaterialParameters(
    Index 3 Param 1=&Multiply_Float(
     A=$Particles.ScaleCurve
     B=&Lerp_Float(
      A=2
      B=1
      Alpha=&FloatFromCurve(FloatCurve=DataInterface() {propertyValues {Curve {keys [
           <at 0 1 RCIM_Cubic>
           {interpMode=RCIM_Cubic, tangentMode=RCTM_User, time=0.3, value=0}
           {arriveTangent=0.5996, interpMode=RCIM_Cubic, leaveTangent=0.5996, tangentMode=RCTM_User, time=0.6, value=0.05}
           {arriveTangent=2.8778, interpMode=RCIM_Cubic, leaveTangent=2.8778, tangentMode=RCTM_User, time=0.96, value=1}
          ]}}})
     )
    )
    Index 3 Param 2=&FloatFromCurve(
     FloatCurve=DataInterface() {propertyValues {Curve {keys [
         {arriveTangent=0.3228, interpMode=RCIM_Cubic, leaveTangent=0.3228, tangentMode=RCTM_User, time=0, value=0}
         {arriveTangent=1.771, interpMode=RCIM_Cubic, leaveTangent=1.771, tangentMode=RCTM_User, time=0.1953, value=0.2}
         {arriveTangent=2.2974, interpMode=RCIM_Cubic, leaveTangent=2.2974, time=0.343, value=1.05}
         {arriveTangent=-2.6609, interpMode=RCIM_Cubic, leaveTangent=-2.6609, time=0.5, value=0.9}
         {arriveTangent=-1.0556, interpMode=RCIM_Cubic, leaveTangent=-1.0556, tangentMode=RCTM_User, time=0.7, value=0.1}
         {tangentMode=RCTM_User, time=1, value=0}
        ]}}}
     Scale Curve=40
    )
    Index 3 Param 3=&RandomRangeFloat(Minimum=0, Maximum=1)
    Index 3 Param 4=1
    Module.Index3Param0Write=true
    Module.Index3Param1Write=true
    Module.Index3Param2Write=true
    Module.Index3Param3Write=true
    Write Parameter Index 0=false
    Write Parameter Index 3=true
   )
  ]
  rendererClasses [&NiagaraMeshRendererProperties]
  renderers [
   {rendererClass=&NiagaraMeshRendererProperties, rendererIndex=0}
   {rendererClass=&NiagaraMeshRendererProperties, rendererIndex=1}
  ]
 }
 Trail {
  emitterUpdateScript [
   &EmitterState(
    Life Cycle Mode=::NewEnumerator1
    Loop Behavior=::NewEnumerator1
    Loop Delay=&Add_Float(A=0.25, B=$System.FXDelay)
    Loop Duration=1.2
    UseLoopDelay=true
   )
   &SpawnPerUnit(
    Module.Use Max Movement Threshold=false
    Module.Use Movement Tolerance=false
    Movement Tolerance=20
    Spawn Spacing:NiagaraFloat=&Lerp_Float(
     A=20
     B=100
     Alpha=&FloatFromCurve(
      FloatCurve=DataInterface() {propertyValues {Curve {keys [
          {interpMode=RCIM_Cubic, tangentMode=RCTM_User, time=0.2, value=0}
          {arriveTangent=5.1366, interpMode=RCIM_Cubic, leaveTangent=5.1366, tangentMode=RCTM_User, time=0.5, value=1}
         ]}}}
      CurveIndex=$Emitter.Age
     )
    )
    Use Spawn Probability=false
   )
  ]
  particleSpawnScript [
   &InitializeParticle(
    Lifetime=1.5
    Mass Mode=::NewEnumerator3
    Ribbon Width=400
    Ribbon Width Mode=::NewEnumerator1
    Sprite Rotation Mode=::NewEnumerator2
    Sprite Size Mode=::NewEnumerator3
    Uniform Sprite Size=50
    UsePositionOffset=false
    Write Material Random=true
   )
   &SkeletalMeshLocation(
    Bone Sampling Mode=::NewEnumerator4
    MaskSpawnGroup=false
    Skeletal Mesh=DataInterface() {propertyValues {
      DefaultMesh=null
      ExcludeBoneName=root
      OptionalComponentClass=null
      PreviewMesh=&SK_M_MALE_Base_Skeleton
      SkinningMode=SkinOnTheFly
      SoftSourceActor=null
      SourceMode=Default
      UvSetIndex=0
      WholeMeshLOD=-1
      bExcludeBone=false
      bReadDeformedGeometry=true
      bRequireCurrentFrameData=true
      ComponentTags []
      FilteredBones [pelvis]
      FilteredSockets []
      MeshUserParameter {parameter {name=None, typeDefHandle {registeredTypeIndex=1}, varData []}}
      SamplingRegions []
     }}
   )
   &AddVelocity(Velocity=<0 0 20>)
   &InheritVelocity(Inherited Velocity Speed Limit=200, Speed Limit=true, UseSpeedThreshold=false)
  ]
  particleUpdateScript [
   &ParticleState()
   &ScaleColor(
    Scale Alpha=&FloatFromCurve(
     FloatCurve=DataInterface() {propertyValues {Curve {keys [<at 0 0 RCIM_Cubic>,<at 0.13 1 RCIM_Cubic>,<at 0.8 0 RCIM_Cubic>]}}}
     Scale Curve=0.8
    )
    Scale RGB=&MakeVectorFromLinearColorRGB(Color=&ColorFromCurve(ColorCurve=DataInterface() {propertyValues {
        AlphaCurve {keys [<at 0 1 RCIM_Cubic>]}
        BlueCurve {keys [<at 0.0141 0.2767 RCIM_Cubic>,<at 0.0609 0.1106>,<at 0.1734 0.0345>,<at 0.7509 0 RCIM_Cubic>]}
        GreenCurve {keys [<at 0.0141 0.2475 RCIM_Cubic>,<at 0.0609 0.0955>,<at 0.1734 0.03>,<at 0.7509 0 RCIM_Cubic>]}
        RedCurve {keys [<at 0.0141 2 RCIM_Cubic>,<at 0.0609 1>,<at 0.1734 0.3>,<at 0.7509 0 RCIM_Cubic>]}
       }}))
    ScaleA=true
    ScaleRGB=true
   )
   &CurlNoiseForce(
    Bias Noise=false
    Noise Frequency=2
    Noise Strength=&Multiply_Float(
     A=220
     B=&FloatFromCurve(FloatCurve=DataInterface() {propertyValues {Curve {keys [{interpMode=RCIM_Cubic, tangentMode=RCTM_User, time=0, value=1}, {interpMode=RCIM_Cubic, tangentMode=RCTM_User, time=0.7, value=0}]}}})
    )
    Pan Noise=false
   )
   &Drag(Drag=4, Use Linear Drag=true, Use Rotational Drag=true)
   &SolveForcesAndVelocity(Clamp Velocity=false, Limit Acceleration=false)
   &DynamicMaterialParameters(
    Index 0 Param 1=&FloatFromCurve(FloatCurve=DataInterface() {propertyValues {Curve {keys [
         {arriveTangent=-1.2682, interpMode=RCIM_Cubic, leaveTangent=-1.2682, tangentMode=RCTM_User, time=0.1, value=1}
         {interpMode=RCIM_Cubic, tangentMode=RCTM_User, time=0.8, value=-0.2}
        ]}}})
    Index 0 Param 2=1
    Index 0 Param 3=0.2
    Module.Index0Param0Write=true
    Module.Index0Param1Write=true
    Module.Index0Param2Write=true
    Module.Index0Param3Write=false
   )
   &ScaleRibbonWidth(Ribbon Width Scale=&FloatFromCurve(FloatCurve=DataInterface() {propertyValues {Curve {keys [
         {arriveTangent=1.8307, interpMode=RCIM_Cubic, leaveTangent=1.8307, tangentMode=RCTM_User, time=0, value=0.2}
         {arriveTangent=0.3065, interpMode=RCIM_Cubic, leaveTangent=0.3065, tangentMode=RCTM_User, time=0.75, value=1}
        ]}}}))
  ]
  rendererClasses [&NiagaraRibbonRendererProperties]
  renderers [{rendererClass=&NiagaraRibbonRendererProperties}]
 }
 TwistAbsorb {
  emitterUpdateScript [
   &EmitterState(
    Life Cycle Mode=::NewEnumerator1
    Loop Behavior=::NewEnumerator1
    Loop Delay=&Add_Float(A=0.02, B=$System.FXDelay)
    Loop Duration=2
    UseLoopDelay=true
   )
   &SpawnBurst_Instantaneous(Use Spawn Probability=false, UseLoopCountLimit=false)
  ]
  particleSpawnScript [
   &InitializeParticle(
    Color=<rgba 5 0.2 0.76 1>
    Lifetime=0.35
    Position Offset=$System.HandFXOffset
    Sprite Rotation Mode=::NewEnumerator2
    Sprite Size Mode=::NewEnumerator3
    Uniform Sprite Size=350
    UsePositionOffset=false
    Write Material Random=true
   )
   &CameraOffset(Camera Offset Amount=110)
  ]
  particleUpdateScript [
   &ParticleState()
   &SkeletalMeshLocation(
    Bone Sampling Mode=::NewEnumerator4
    MaskSpawnGroup=false
    Orientation Sampling=::NewEnumerator1
    Position Sampling=::NewEnumerator1
    Sampled Position Offset=<50 0 0>
    Skeletal Mesh=DataInterface() {propertyValues {
      DefaultMesh=null
      ExcludeBoneName=root
      OptionalComponentClass=null
      PreviewMesh=&SK_M_MALE_Base_Skeleton
      SkinningMode=SkinOnTheFly
      SoftSourceActor=null
      SourceMode=Default
      UvSetIndex=0
      WholeMeshLOD=-1
      bExcludeBone=false
      bReadDeformedGeometry=true
      bRequireCurrentFrameData=true
      ComponentTags []
      FilteredBones [pelvis]
      FilteredSockets []
      MeshUserParameter {parameter {name=None, typeDefHandle {registeredTypeIndex=1}, varData []}}
      SamplingRegions []
     }}
    Source Movement Scale:Vector3f=&VectorFromFloat(Value=&FloatFromCurve(
      FloatCurve=DataInterface() {propertyValues {Curve {keys [{interpMode=RCIM_Cubic, tangentMode=RCTM_User, time=0.42, value=1}, {interpMode=RCIM_Cubic, tangentMode=RCTM_User, time=0.44, value=0}]}}}
      CurveIndex=$Emitter.Age
     ))
   )
   &ScaleSpriteSize(Uniform Curve Sprite Scale=DataInterface() {propertyValues {Curve {keys [
        {arriveTangent=12.5362, interpMode=RCIM_Cubic, leaveTangent=12.5362, tangentMode=RCTM_User, time=0, value=0.3}
        {arriveTangent=0.2602, interpMode=RCIM_Cubic, leaveTangent=0.2602, tangentMode=RCTM_User, time=0.1, value=1}
       ]}}})
   &SolveForcesAndVelocity(Clamp Velocity=false, Limit Acceleration=false)
   &DynamicMaterialParameters(
    Index 0 Param 1=&Lerp_Float(
     A=-0.2
     B=1
     Alpha=&FloatFromCurve(FloatCurve=DataInterface() {propertyValues {Curve {keys [
          {arriveTangent=0.1556, interpMode=RCIM_Cubic, leaveTangent=0.1556, tangentMode=RCTM_User, time=0, value=0}
          {arriveTangent=2.2716, interpMode=RCIM_Cubic, leaveTangent=2.2716, tangentMode=RCTM_User, time=0.8, value=1}
         ]}}})
    )
    Index 0 Param 2=1
    Index 0 Param 3=&Lerp_Float(
     A=-0.1
     B=1
     Alpha=&FloatFromCurve(FloatCurve=DataInterface() {propertyValues {Curve {keys [
          {arriveTangent=0.1235, interpMode=RCIM_Cubic, leaveTangent=0.1235, tangentMode=RCTM_User, time=0.2, value=0}
          {arriveTangent=1.955, interpMode=RCIM_Cubic, leaveTangent=1.955, tangentMode=RCTM_User, time=1, value=1}
         ]}}})
    )
    Index 0 Param 4=&Lerp_Float(
     A=0
     B=0.25
     Alpha=&FloatFromCurve(FloatCurve=DataInterface() {propertyValues {Curve {keys [
          {arriveTangent=0.3497, interpMode=RCIM_Cubic, leaveTangent=0.3497, tangentMode=RCTM_User, time=0, value=0}
          {interpMode=RCIM_Cubic, tangentMode=RCTM_User, time=1, value=1}
         ]}}})
    )
    Module.Index0Param0Write=true
    Module.Index0Param1Write=true
    Module.Index0Param2Write=true
    Module.Index0Param3Write=true
   )
   &ScaleColor(
    Scale Alpha=&FloatFromCurve(
     FloatCurve=DataInterface() {propertyValues {Curve {keys [
         <at 0 0 RCIM_Cubic>
         {arriveTangent=-0.0178, interpMode=RCIM_Cubic, leaveTangent=-0.0178, tangentMode=RCTM_User, time=0.1, value=1}
         {interpMode=RCIM_Cubic, tangentMode=RCTM_User, time=0.3, value=1}
         {interpMode=RCIM_Cubic, tangentMode=RCTM_User, time=1, value=0}
        ]}}}
     Scale Curve=2
    )
    Scale RGB=&MakeVectorFromLinearColorRGB(Color=&ColorFromCurve(ColorCurve=DataInterface() {propertyValues {
        AlphaCurve {keys [<at 0 1 RCIM_Cubic>]}
        BlueCurve {keys [<at 0.1216 1 RCIM_Cubic>,<at 0.8838 0.0023 RCIM_Cubic>]}
        GreenCurve {keys [<at 0.1216 1 RCIM_Cubic>,<at 0.8838 0.0018 RCIM_Cubic>]}
        RedCurve {keys [<at 0.1216 1 RCIM_Cubic>,<at 0.8838 0.03 RCIM_Cubic>]}
       }}))
    ScaleA=true
    ScaleRGB=true
   )
   /SetVariables(
    Particles.ColorDark:LinearColor=&MultiplyLinearColor(
     Linear Color=<rgba 0.045 0 0.0007 1>
     Scale Factor=&MakeFloat4FromVectorAndFloat(
      "Vector (RGB)"=<1 1 1>
      "Float (Alpha)"=&FloatFromCurve(
       FloatCurve=DataInterface() {propertyValues {Curve {keys [
           <at 0 0 RCIM_Cubic>
           {arriveTangent=-2.1159, interpMode=RCIM_Cubic, leaveTangent=-2.1159, tangentMode=RCTM_User, time=0.2, value=1}
           {interpMode=RCIM_Cubic, tangentMode=RCTM_User, time=1, value=0}
          ]}}}
       Scale Curve=1.6
      )
     )
    )
    Particles.DynamicMaterialParameterDark:Vector4f=&MakeVector4(
     X=&Lerp_Float(
      A=0
      B=1
      Alpha=&FloatFromCurve(FloatCurve=DataInterface() {propertyValues {Curve {keys [
           {interpMode=RCIM_Cubic, tangentMode=RCTM_User, time=0, value=0}
           {arriveTangent=1.7687, interpMode=RCIM_Cubic, leaveTangent=1.7687, tangentMode=RCTM_User, time=1, value=1}
          ]}}})
     )
     Y=1
     Z=&Lerp_Float(
      A=-0.5
      B=1
      Alpha=&FloatFromCurve(FloatCurve=DataInterface() {propertyValues {Curve {keys [
           {arriveTangent=0.1235, interpMode=RCIM_Cubic, leaveTangent=0.1235, tangentMode=RCTM_User, time=0.2, value=0}
           {arriveTangent=1.955, interpMode=RCIM_Cubic, leaveTangent=1.955, tangentMode=RCTM_User, time=1, value=1}
          ]}}})
     )
     W=0.232
    )
    Particles.SpriteSizeDark:Vector2f=&Multiply_Vector2D_ByFloat(Float=1.4, Vector2D=$Particles.SpriteSize)
   ) {isSetParameters=true}
  ]
  rendererClasses [&NiagaraSpriteRendererProperties]
  renderers [
   {rendererClass=&NiagaraSpriteRendererProperties}
   {rendererClass=&NiagaraSpriteRendererProperties, rendererIndex=1}
  ]
 }
}
summary {
 systemName=NS_MaskMaker_Teleport_Start
 emitters [
  {emitterName=TwistAbsorb, rendererClasses [&NiagaraSpriteRendererProperties]}
  {emitterName=PortalBig, rendererClasses [&NiagaraMeshRendererProperties]}
  {emitterName=Trail, rendererClasses [&NiagaraRibbonRendererProperties]}
  {emitterName=FlareSmoke, rendererClasses [&NiagaraSpriteRendererProperties, &NiagaraMeshRendererProperties]}
 ]
 userVariables [
  User.GlobalOpacity {type=NiagaraFloat, defaultValue=1, description=""}
  User.GlobalScale {type=NiagaraFloat, defaultValue=1, description=""}
  User.Offset Pos {type=NiagaraPosition, description="", defaultValue {}}
  User.RibbonIDOffset {type=NiagaraInt32}
  User.WanderingBoss {type=NiagaraBool}
 ]
}
